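# Physics asset for the belt on the M_MED_RampCanal_Bar male player cosmetic body, giving four belt pieces (front and back, left and right) collision and jointed motion.
# Each piece chains from a spine_02 torso capsule through a tiny base capsule to a larger dyn capsule.
# Base-to-spine joints rotate freely, while dyn-to-base joints have zero twist and narrow swing, confined to one axis on all but front right.
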PhysicsAsset /BRCosmetics/Characters/Player/Male/Medium/Bodies/M_MED_RampCanal_Bar/Meshes/M_MED_RampCanal_Bar_Belt_Physics
bodies {
 base_belt_back_lf [{shapeType=Capsule, length=0.505, radius=0.505, shapeName=base_belt_back_lf_capsule, rotation=<rot 0 -90 0>}]
 base_belt_back_rt [{shapeType=Capsule, length=0.505, radius=0.505, shapeName=base_belt_back_rt_capsule, rotation=<rot 0 -90 0>}]
 base_belt_front_lf [{shapeType=Capsule, length=0.505, radius=0.505, shapeName=base_belt_front_lf_capsule, rotation=<rot 0 -90 0>}]
 base_belt_front_rt [{shapeType=Capsule, length=0.505, radius=0.505, shapeName=base_belt_front_rt_capsule, rotation=<rot 0 -90 0>}]
 dyn_belt_back_lf [{
   shapeType=Capsule
   length=11.428
   radius=2.5
   shapeName=dyn_belt_back_lf_capsule
   center=<7.9363 6.6389 4.1509>
   rotation=<rot 0 126.0537 -60.3355>
  }]
 dyn_belt_back_rt [{
   shapeType=Capsule
   length=12.6574
   radius=2.5
   shapeName=dyn_belt_back_rt_capsule
   center=<7.7279 6.9623 -4.1059>
   rotation=<rot 0 -52.2728 -59.6682>
  }]
 dyn_belt_front_lf [{
   shapeType=Capsule
   length=12.4725
   radius=4.1443
   shapeName=dyn_belt_front_lf_capsule
   center=<2.1324 -3.4241 -9.9355>
   rotation=<rot -5.5157 -172.9287 -80.3716>
  }]
 dyn_belt_front_rt [{
   shapeType=Capsule
   length=9.0099
   radius=2.4634
   shapeName=dyn_belt_front_rt_capsule
   center=<1.4188 5.7028 -11.9247>
   rotation=<rot 0 -1.5816 -50.5231>
  }]
 spine_02 [{
   shapeType=Capsule
   length=3.6249
   radius=9.8954
   shapeName=spine_02_capsule
   center=<3.5937 1.4815 0.2357>
   rotation=<rot -12.5669 -90 1.2876>
  }]
}
constraints [
 {
  bone1Name=base_belt_front_lf
  bone2Name=spine_02
  swing1LimitDegrees=45
  swing1Motion=Free
  swing2LimitDegrees=45
  swing2Motion=Free
  twistLimitDegrees=45
  twistMotion=Free
 }
 {bone1Name=dyn_belt_front_lf, bone2Name=base_belt_front_lf, swing1LimitDegrees=0, swing2LimitDegrees=20, twistLimitDegrees=0}
 {
  bone1Name=base_belt_front_rt
  bone2Name=spine_02
  swing1LimitDegrees=45
  swing1Motion=Free
  swing2LimitDegrees=45
  swing2Motion=Free
  twistLimitDegrees=45
  twistMotion=Free
 }
 {bone1Name=dyn_belt_front_rt, bone2Name=base_belt_front_rt, swing1LimitDegrees=20, swing2LimitDegrees=20, twistLimitDegrees=0}
 {
  bone1Name=base_belt_back_lf
  bone2Name=spine_02
  swing1LimitDegrees=45
  swing1Motion=Free
  swing2LimitDegrees=45
  swing2Motion=Free
  twistLimitDegrees=45
  twistMotion=Free
 }
 {bone1Name=dyn_belt_back_lf, bone2Name=base_belt_back_lf, swing1LimitDegrees=20, swing2LimitDegrees=0, twistLimitDegrees=0}
 {
  bone1Name=base_belt_back_rt
  bone2Name=spine_02
  swing1LimitDegrees=45
  swing1Motion=Free
  swing2LimitDegrees=45
  swing2Motion=Free
  twistLimitDegrees=45
  twistMotion=Free
 }
 {bone1Name=dyn_belt_back_rt, bone2Name=base_belt_back_rt, swing1LimitDegrees=20, swing2LimitDegrees=0, twistLimitDegrees=0}
]
