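# Control Rig for the M_MED_RampCanal_Bar player body, driving deform, back-belt, coat and foot helper bones.
# The shared DeformRig function runs first, then belt parent constraints, then an IsCloth branch.
# Without cloth, spherical pose readers on thighs and pelvis, smoothed crouch/jump flags, pelvis height and foot rotation weight corrective transforms.
# With cloth, only pelvis_cloth_helper aims at root while grounded; the coat_cr rotation constraint is unwired.
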ControlRigBlueprint /BRCosmetics/Characters/Player/Male/Medium/Bodies/M_MED_RampCanal_Bar/Meshes/M_MED_RampCanal_Bar_CtrlRig
graphs [
 New Function_ContainedGraph {nodes [
   RigVMFunctionEntryNode()
   > RigVMFunctionReturnNode()
  ]}
 New Function_ContainedGraph {nodes [
   n1 = RigVMFunctionEntryNode()
   RigVMFunction_MathQuaternionToEuler = &RigVMFunction_MathQuaternionToEuler(
    RotationOrder=ZYX
    Value=&RigUnit_GetRelativeTransformForItem(
     Child=(Type=Bone,Name="hand_l")
     Parent=(Type=Bone,Name="lowerarm_l")
     bChildInitial=False
     bParentInitial=False
     Child.Name=@n1.Child
     Parent.Name=@n1.Parent
    ).RelativeTransform.Rotation
   )
   Remap = &RigVMFunction_MathDoubleRemap(
    TargetMaximum=1
    TargetMinimum=0
    Value=@RigVMFunction_MathQuaternionToEuler.Result.Z
    SourceMaximum=@n1.Source Maximum
    SourceMinimum=@n1.Source Minimum
   )
   > RigVMFunctionReturnNode(Delta=@Remap.Result)
  ]}
 RigVMFunctionLibrary {nodes [
   Delta Rotation = RigVMCollapseNode(ExecuteContext=())
   New Function = RigVMCollapseNode(ExecuteContext=())
  ]}
 RigVMFunction_Sequence_1_ContainedGraph {nodes [
   n1 = RigVMFunctionEntryNode()
   RigVMFunction_Sequence = &RigVMFunction_Sequence(ExecuteContext=@n1.ExecuteContext)
   n2 = &RigVMFunction_Sequence(ExecuteContext=@RigVMFunction_Sequence.B)
   RigVMFunctionReturnNode(A=@RigVMFunction_Sequence.A, B=@n2.A, C=@n2.B)
  ]}
 RigVMModel {nodes [
   RigUnit_BeginExecution = &RigUnit_BeginExecution()
   n1 = &RigVMFunction_Sequence(ExecuteContext=@RigUnit_BeginExecution.ExecutePin)
   FN_DeformRig_FNC = RigVMFunctionReferenceNode(
    DeformRig_ Curve="(deformRig_l_upperarm_up_alpha_72_#1=False,deformRig_r_upperarm_up_alpha_73_#2=False,deformRig_l_upperarm_fr_alpha_74_#3=False,deformRig_r_upperarm_fr_alpha_75_CB51B8 ...(truncated)"
    DeformRig_Bool="(upperarm_1_#4=True,elbow_3_#5=True,wrist_7_#6=True,groin_8_#7=True,glute_17_#8=True,knee_9_40C408324CA261 ...(truncated)"
    DeformRig_Debug=(upperarm_debug_18_#4=False,wrist_debug_20_#6=False,lat_debug_26_#9=False,groinGlute_debug_27_#7=False)
    DeformRig_Value="(upperarm_up_18_#10=(Rotation=(X=0.0,Y=-0.0,Z=0.0,W=1.0),Translation=(X=0.0,Y=0.0,Z=8.0),Scale3D=(X=0.70,Y=1.0,Z=1.0)),upperarm_fwd_fr_19_#11=(Rotation=(X=0.0,Y=-0.0,Z=0.0,W=1.0),Translation=(X=0.0, ...(truncated)"
    ExecuteContext=@n1.A
   )
   > &RigUnit_ParentConstraint(
    AdvancedSettings=(RotationOrderForFilter=XZY)
    Child=(Type=Bone,Name="base_belt_back_lf")
    Parents="((Item=(Type=Bone,Name=\"clavicle_l\"),Weight=0.025),(Item=(Type=Bone,Name=\"clavicle_r\"),Weight=0.0254),(Item=(Type=Bone,Name=\"pelvis\"),Weight=0.0343),(Item=(Type=Bone,Name=\"spine_01\"),Weight=0.0673),(Item=(Type=Bone,Name=\"spine_05\"),Weight=0.0806),(Item=(Ty ...(truncated)"
    Weight=1
   )
   > &RigUnit_ParentConstraint(
    AdvancedSettings=(RotationOrderForFilter=XZY)
    Child=(Type=Bone,Name="base_belt_back_rt")
    Parents="((Item=(Type=Bone,Name=\"clavicle_l\"),Weight=0.025),(Item=(Type=Bone,Name=\"clavicle_r\"),Weight=0.0254),(Item=(Type=Bone,Name=\"pelvis\"),Weight=0.0343),(Item=(Type=Bone,Name=\"spine_01\"),Weight=0.0673),(Item=(Type=Bone,Name=\"spine_05\"),Weight=0.0806),(Item=(Ty ...(truncated)"
    Weight=1
   )
   n2 = &RigVMFunction_ControlFlowBranch(ExecuteContext=@n1.B, Condition=RigVMVariableNode() {variable=IsCloth}.Value)
   RigVMFunction_Sequence_2 = RigVMAggregateNode(ExecuteContext=@n2.False)
   n3 = &RigUnit_SphericalPoseReader(
    ActiveRegionScaleFactors=(PositiveWidth=1,NegativeWidth=1,PositiveHeight=1,NegativeHeight=1)
    ActiveRegionSize=0.05
    Debug=(bDrawDebug=false,bDraw2D=False,bDrawLocalAxes=False,DebugScale=25,DebugSegments=20,DebugThickness=0.25)
    DriverAxis=(X=-1,Y=0,Z=0)
    DriverItem=(Type=Bone,Name="thigh_l")
    FalloffRegionScaleFactors=(PositiveWidth=1,NegativeWidth=1,PositiveHeight=0.9,NegativeHeight=0.9)
    FalloffSize=0.63
    FlipHeightScaling=False
    FlipWidthScaling=False
    OptionalParentItem=(Type=Bone,Name="None")
    RotationOffset=(X=90,Y=0,Z=-30)
    ExecutePin=@RigVMFunction_Sequence_2.A
   )
   > &RigUnit_ModifyTransforms(
    ItemToModify="((Item=(Type=Bone,Name=\"base_coat_front_lf\"),Transform=(Rotation=(X=-0.1155,Y=0.1453,Z=-0.1646,W=-0.9687),Translation=(X=1.9912,Y=2.6329,Z=-4.9543),Scale3D=(X=1,Y=1,Z=1))),(Item=(Type=Bone,Name=\"base_coat_front_side_lf\"),Transform=(Rotation=(X=0,Y=-0.1305, ...(truncated)"
    WeightMinimum=0
    Weight=@n3.OutputParam
   )
   > n4 = &RigUnit_SphericalPoseReader(
    ActiveRegionScaleFactors=(PositiveWidth=1,NegativeWidth=1,PositiveHeight=1,NegativeHeight=1)
    ActiveRegionSize=0.05
    Debug=(bDrawDebug=false,bDraw2D=False,bDrawLocalAxes=False,DebugScale=25,DebugSegments=20,DebugThickness=0.25)
    DriverAxis=(X=-1,Y=0,Z=0)
    DriverItem=(Type=Bone,Name="thigh_l")
    FalloffRegionScaleFactors=(PositiveWidth=1,NegativeWidth=1,PositiveHeight=0.7,NegativeHeight=0.7)
    FalloffSize=0.63
    FlipHeightScaling=False
    FlipWidthScaling=False
    OptionalParentItem=(Type=Bone,Name="None")
    RotationOffset=(X=180,Y=30,Z=0)
   )
   > &RigUnit_ModifyTransforms(
    ItemToModify="((Item=(Type=Bone,Name=\"base_coat_front_lf\"),Transform=(Rotation=(X=0.1391,Y=-0.519,Z=0.2183,W=-0.8147),Translation=(X=0,Y=0,Z=0),Scale3D=(X=1,Y=1,Z=1))),(Item=(Type=Bone,Name=\"base_coat_front_side_lf\"),Transform=(Rotation=(X=-0.0643,Y=-0.0589,Z=0.7345,W=- ...(truncated)"
    WeightMinimum=0
    Weight=@n4.OutputParam
   )
   > n5 = &RigUnit_SphericalPoseReader(
    ActiveRegionScaleFactors=(PositiveWidth=1,NegativeWidth=1,PositiveHeight=1,NegativeHeight=1)
    ActiveRegionSize=0.065
    Debug=(bDrawDebug=false,bDraw2D=False,bDrawLocalAxes=False,DebugScale=25,DebugSegments=20,DebugThickness=0.25)
    DriverAxis=(X=-1,Y=0,Z=0)
    DriverItem=(Type=Bone,Name="thigh_l")
    FalloffRegionScaleFactors=(PositiveWidth=0.9,NegativeWidth=0.93,PositiveHeight=0.9,NegativeHeight=0.9)
    FalloffSize=0.65
    FlipHeightScaling=False
    FlipWidthScaling=False
    OptionalParentItem=(Type=Bone,Name="None")
    RotationOffset=(X=140,Y=30,Z=0)
   )
   > &RigUnit_ModifyTransforms(
    ItemToModify="((Item=(Type=Bone,Name=\"base_coat_front_lf\"),Transform=(Rotation=(X=-0.3963,Y=-0.2446,Z=0.5481,W=-0.6948),Translation=(X=-2.9945,Y=-0.342,Z=-3.2252),Scale3D=(X=1,Y=1,Z=1))),(Item=(Type=Bone,Name=\"base_coat_front_side_lf\"),Transform=(Rotation=(X=0,Y=0,Z=0.1 ...(truncated)"
    WeightMinimum=0
    Weight=@n5.OutputParam
   )
   > n6 = &RigUnit_SphericalPoseReader(
    ActiveRegionScaleFactors=(PositiveWidth=1,NegativeWidth=1,PositiveHeight=1,NegativeHeight=1)
    ActiveRegionSize=0.05
    Debug=(bDrawDebug=false,bDraw2D=False,bDrawLocalAxes=False,DebugScale=25,DebugSegments=20,DebugThickness=0.25)
    DriverAxis=(X=-1,Y=0,Z=0)
    DriverItem=(Type=Bone,Name="thigh_l")
    FalloffRegionScaleFactors=(PositiveWidth=0.75,NegativeWidth=0.75,PositiveHeight=0.9,NegativeHeight=0.9)
    FalloffSize=0.63
    FlipHeightScaling=False
    FlipWidthScaling=False
    OptionalParentItem=(Type=Bone,Name="None")
    RotationOffset=(X=-90,Y=0,Z=10)
   )
   > &RigUnit_ModifyTransforms(
    ItemToModify="((Item=(Type=Bone,Name=\"base_coat_back_side_lf\"),Transform=(Rotation=(X=0,Y=0,Z=-0.6088,W=0.7934),Translation=(X=0,Y=0,Z=0),Scale3D=(X=1,Y=1,Z=1))),(Item=(Type=Bone,Name=\"base_coat_back_lf\"),Transform=(Rotation=(X=0,Y=0,Z=-0.7373,W=0.6756),Translation=(X=0 ...(truncated)"
    WeightMinimum=0
    Weight=@n6.OutputParam
   )
   > n7 = &RigUnit_SphericalPoseReader(
    ActiveRegionScaleFactors=(PositiveWidth=1,NegativeWidth=1,PositiveHeight=1,NegativeHeight=1)
    ActiveRegionSize=0.05
    Debug=(bDrawDebug=false,bDraw2D=False,bDrawLocalAxes=False,DebugScale=25,DebugSegments=20,DebugThickness=0.25)
    DriverAxis=(X=-1,Y=0,Z=0)
    DriverItem=(Type=Bone,Name="thigh_l")
    FalloffRegionScaleFactors=(PositiveWidth=1,NegativeWidth=0.75,PositiveHeight=0.9,NegativeHeight=0.9)
    FalloffSize=0.63
    FlipHeightScaling=False
    FlipWidthScaling=false
    OptionalParentItem=(Type=Bone,Name="None")
    RotationOffset=(X=-135,Y=0,Z=0)
   )
   > &RigUnit_ModifyTransforms(
    ItemToModify="((Item=(Type=Bone,Name=\"base_coat_back_side_lf\"),Transform=(Rotation=(X=0.0057,Y=-0.0432,Z=-0.4613,W=0.8862),Translation=(X=0,Y=0,Z=0),Scale3D=(X=1,Y=1,Z=1))),(Item=(Type=Bone,Name=\"base_coat_back_lf\"),Transform=(Rotation=(X=0,Y=0,Z=0,W=1),Translation=(X=0 ...(truncated)"
    WeightMinimum=0
    Weight=@n7.OutputParam
   )
   > n8 = &RigUnit_SphericalPoseReader(
    ActiveRegionScaleFactors=(PositiveWidth=1,NegativeWidth=1,PositiveHeight=1,NegativeHeight=1)
    ActiveRegionSize=0.3
    Debug=(bDrawDebug=false,bDraw2D=False,bDrawLocalAxes=False,DebugScale=25,DebugSegments=20,DebugThickness=0.25)
    DriverAxis=(X=0,Y=1,Z=0)
    DriverItem=(Type=Bone,Name="pelvis")
    FalloffRegionScaleFactors=(PositiveWidth=1,NegativeWidth=1,PositiveHeight=1,NegativeHeight=1)
    FalloffSize=0.45
    FlipHeightScaling=False
    FlipWidthScaling=False
    OptionalParentItem=(Type=Bone,Name="spine_05")
    RotationOffset=(X=0,Y=-90,Z=0)
   )
   Not_Equals = RigVMDispatchNode(
    A=&RigVMFunction_MathBoolNot(Value=RigVMVariableNode() {variable=IsCrouch}.Value).Result
    B=RigVMVariableNode() {variable=IsJumping}.Value
   ) {template="DISPATCH_RigVMDispatch_CoreNotEquals(in A,in B,out Result)"}
   n9 = &RigVMFunction_AlphaInterp(
    InterpSpeedDecreasing=10
    InterpSpeedIncreasing=10
    Scale=1
    bClampResult=true
    bMapRange=true
    Value=&RigVMFunction_MathBoolToFloat(Value=@Not_Equals.Result).Result
   )
   > &RigUnit_ModifyTransforms(
    ItemToModify="((Item=(Type=Bone,Name=\"orient_coat_front_lf\"),Transform=(Rotation=(X=0,Y=0,Z=0,W=1),Translation=(X=6,Y=6,Z=0),Scale3D=(X=1,Y=1,Z=1))),(Item=(Type=Bone,Name=\"orient_coat_front_rt\"),Transform=(Rotation=(X=0,Y=0,Z=0,W=1),Translation=(X=6,Y=6,Z=0),Scale3D=(X= ...(truncated)"
    WeightMinimum=0
    Weight=&RigVMFunction_MathDoubleMul(B=@n9.Result, A=@n8.OutputParam).Result
   )
   > n10 = &RigUnit_SphericalPoseReader(
    ActiveRegionScaleFactors=(PositiveWidth=1,NegativeWidth=1,PositiveHeight=1,NegativeHeight=1)
    ActiveRegionSize=0.3
    Debug=(bDrawDebug=false,bDraw2D=False,bDrawLocalAxes=False,DebugScale=25,DebugSegments=20,DebugThickness=0.25)
    DriverAxis=(X=0,Y=1,Z=0)
    DriverItem=(Type=Bone,Name="pelvis")
    FalloffRegionScaleFactors=(PositiveWidth=1,NegativeWidth=1,PositiveHeight=1,NegativeHeight=1)
    FalloffSize=0.5
    FlipHeightScaling=False
    FlipWidthScaling=False
    OptionalParentItem=(Type=Bone,Name="spine_05")
    RotationOffset=(X=0,Y=0,Z=0)
   )
   n11 = &RigVMFunction_AlphaInterp(
    InterpSpeedDecreasing=10
    InterpSpeedIncreasing=10
    Scale=1
    bClampResult=true
    bMapRange=true
    Value=&RigVMFunction_MathBoolToFloat(Value=RigVMVariableNode() {variable=IsCrouch}.Value).Result
   )
   > &RigUnit_ModifyTransforms(
    ItemToModify="((Item=(Type=Bone,Name=\"offset_coat_side_lf\"),Transform=(Rotation=(X=0,Y=0,Z=0,W=1),Translation=(X=0,Y=-2,Z=0),Scale3D=(X=1,Y=1,Z=1))),(Item=(Type=Bone,Name=\"offset_coat_back_side_lf\"),Transform=(Rotation=(X=0,Y=0,Z=0,W=1),Translation=(X=0,Y=-2,Z=2.3766),S ...(truncated)"
    WeightMinimum=0
    Weight=&RigVMFunction_MathDoubleMul(B=@n11.Result, A=@n10.OutputParam).Result
   )
   > n12 = &RigUnit_SphericalPoseReader(
    ActiveRegionScaleFactors=(PositiveWidth=1,NegativeWidth=1,PositiveHeight=1,NegativeHeight=1)
    ActiveRegionSize=0.3
    Debug=(bDrawDebug=false,bDraw2D=False,bDrawLocalAxes=False,DebugScale=25,DebugSegments=20,DebugThickness=0.25)
    DriverAxis=(X=0,Y=1,Z=0)
    DriverItem=(Type=Bone,Name="pelvis")
    FalloffRegionScaleFactors=(PositiveWidth=1,NegativeWidth=1,PositiveHeight=1,NegativeHeight=1)
    FalloffSize=0.5
    FlipHeightScaling=False
    FlipWidthScaling=False
    OptionalParentItem=(Type=Bone,Name="spine_05")
    RotationOffset=(X=180,Y=0,Z=0)
   )
   > &RigUnit_ModifyTransforms(
    ItemToModify=((Item=(Type=Bone,Name="base_coat_side_rt"),Transform=(Rotation=(X=0.2588,Y=0,Z=0,W=0.9659),Translation=(X=0,Y=0,Z=0),Scale3D=(X=1,Y=1,Z=1))))
    WeightMinimum=0
    Weight=&RigVMFunction_MathDoubleMul(B=@n11.Result, A=@n12.OutputParam).Result
   )
   n13 = &RigUnit_SphericalPoseReader(
    ActiveRegionScaleFactors=(PositiveWidth=1,NegativeWidth=1,PositiveHeight=1,NegativeHeight=1)
    ActiveRegionSize=0.05
    Debug=(bDrawDebug=false,bDraw2D=False,bDrawLocalAxes=False,DebugScale=25,DebugSegments=20,DebugThickness=0.25)
    DriverAxis=(X=1,Y=0,Z=0)
    DriverItem=(Type=Bone,Name="thigh_r")
    FalloffRegionScaleFactors=(PositiveWidth=1,NegativeWidth=1,PositiveHeight=0.9,NegativeHeight=0.9)
    FalloffSize=0.63
    FlipHeightScaling=true
    FlipWidthScaling=true
    OptionalParentItem=(Type=Bone,Name="None")
    RotationOffset=(X=-90,Y=0,Z=-30)
    ExecutePin=@RigVMFunction_Sequence_2.B
   )
   > &RigUnit_ModifyTransforms(
    ItemToModify="((Item=(Type=Bone,Name=\"base_coat_front_rt\"),Transform=(Rotation=(X=-0.1769,Y=0.1076,Z=0.7862,W=0.5822),Translation=(X=-1.6917,Y=-2.6047,Z=3.5113),Scale3D=(X=1,Y=1,Z=1))),(Item=(Type=Bone,Name=\"base_coat_front_side_rt\"),Transform=(Rotation=(X=0,Y=0,Z=0.461 ...(truncated)"
    WeightMinimum=0
    Weight=@n13.OutputParam
   )
   > n14 = &RigUnit_SphericalPoseReader(
    ActiveRegionScaleFactors=(PositiveWidth=1,NegativeWidth=1,PositiveHeight=1,NegativeHeight=1)
    ActiveRegionSize=0.05
    Debug=(bDrawDebug=false,bDraw2D=False,bDrawLocalAxes=False,DebugScale=25,DebugSegments=20,DebugThickness=0.25)
    DriverAxis=(X=1,Y=0,Z=0)
    DriverItem=(Type=Bone,Name="thigh_r")
    FalloffRegionScaleFactors=(PositiveWidth=1,NegativeWidth=1,PositiveHeight=0.7,NegativeHeight=0.7)
    FalloffSize=0.63
    FlipHeightScaling=False
    FlipWidthScaling=False
    OptionalParentItem=(Type=Bone,Name="None")
    RotationOffset=(X=0,Y=30,Z=0)
   )
   > &RigUnit_ModifyTransforms(
    ItemToModify="((Item=(Type=Bone,Name=\"base_coat_front_rt\"),Transform=(Rotation=(X=0,Y=0,Z=0.0436,W=0.999),Translation=(X=0,Y=0,Z=0),Scale3D=(X=1,Y=1,Z=1))),(Item=(Type=Bone,Name=\"base_coat_front_side_rt\"),Transform=(Rotation=(X=0,Y=0,Z=0.6756,W=-0.7373),Translation=(X=0 ...(truncated)"
    WeightMinimum=0
    Weight=@n14.OutputParam
   )
   > n15 = &RigUnit_SphericalPoseReader(
    ActiveRegionScaleFactors=(PositiveWidth=1,NegativeWidth=1,PositiveHeight=1,NegativeHeight=1)
    ActiveRegionSize=0.065
    Debug=(bDrawDebug=false,bDraw2D=False,bDrawLocalAxes=False,DebugScale=25,DebugSegments=20,DebugThickness=0.25)
    DriverAxis=(X=1,Y=0,Z=0)
    DriverItem=(Type=Bone,Name="thigh_r")
    FalloffRegionScaleFactors=(PositiveWidth=0.93,NegativeWidth=0.9,PositiveHeight=0.9,NegativeHeight=0.9)
    FalloffSize=0.65
    FlipHeightScaling=False
    FlipWidthScaling=False
    OptionalParentItem=(Type=Bone,Name="None")
    RotationOffset=(X=-40,Y=30,Z=0)
   )
   > &RigUnit_ModifyTransforms(
    ItemToModify="((Item=(Type=Bone,Name=\"offset_coat_front_rt\"),Transform=(Rotation=(X=0,Y=0,Z=-0.999,W=0.0436),Translation=(X=-2.6729,Y=0.9728,Z=0),Scale3D=(X=1,Y=1,Z=1))),(Item=(Type=Bone,Name=\"offset_coat_front_side_rt\"),Transform=(Rotation=(X=-0.1091,Y=0.0584,Z=-0.7483 ...(truncated)"
    WeightMinimum=0
    Weight=@n15.OutputParam
   )
   > n16 = &RigUnit_SphericalPoseReader(
    ActiveRegionScaleFactors=(PositiveWidth=1,NegativeWidth=1,PositiveHeight=1,NegativeHeight=1)
    ActiveRegionSize=0.05
    Debug=(bDrawDebug=false,bDraw2D=False,bDrawLocalAxes=False,DebugScale=25,DebugSegments=20,DebugThickness=0.25)
    DriverAxis=(X=1,Y=0,Z=0)
    DriverItem=(Type=Bone,Name="thigh_r")
    FalloffRegionScaleFactors=(PositiveWidth=0.75,NegativeWidth=0.75,PositiveHeight=0.9,NegativeHeight=0.9)
    FalloffSize=0.63
    FlipHeightScaling=False
    FlipWidthScaling=False
    OptionalParentItem=(Type=Bone,Name="None")
    RotationOffset=(X=90,Y=0,Z=10)
   )
   > &RigUnit_ModifyTransforms(
    ItemToModify="((Item=(Type=Bone,Name=\"base_coat_back_side_rt\"),Transform=(Rotation=(X=0,Y=0,Z=-0.766,W=0.6428),Translation=(X=0,Y=0,Z=0),Scale3D=(X=1,Y=1,Z=1))),(Item=(Type=Bone,Name=\"base_coat_back_rt\"),Transform=(Rotation=(X=0,Y=0,Z=-0.9397,W=0.342),Translation=(X=0,Y ...(truncated)"
    WeightMinimum=0
    Weight=@n16.OutputParam
   )
   > n17 = &RigUnit_SphericalPoseReader(
    ActiveRegionScaleFactors=(PositiveWidth=1,NegativeWidth=1,PositiveHeight=1,NegativeHeight=1)
    ActiveRegionSize=0.05
    Debug=(bDrawDebug=false,bDraw2D=False,bDrawLocalAxes=False,DebugScale=25,DebugSegments=20,DebugThickness=0.25)
    DriverAxis=(X=1,Y=0,Z=0)
    DriverItem=(Type=Bone,Name="thigh_r")
    FalloffRegionScaleFactors=(PositiveWidth=0.75,NegativeWidth=1,PositiveHeight=0.9,NegativeHeight=0.9)
    FalloffSize=0.63
    FlipHeightScaling=False
    FlipWidthScaling=false
    OptionalParentItem=(Type=Bone,Name="None")
    RotationOffset=(X=45,Y=0,Z=0)
   )
   > &RigUnit_ModifyTransforms(
    ItemToModify="((Item=(Type=Bone,Name=\"base_coat_back_side_rt\"),Transform=(Rotation=(X=0,Y=0,Z=-0.0872,W=0.9962),Translation=(X=0,Y=0,Z=0),Scale3D=(X=1,Y=1,Z=1))),(Item=(Type=Bone,Name=\"base_coat_back_rt\"),Transform=(Rotation=(X=0,Y=0,Z=-0.0872,W=0.9962),Translation=(X=0 ...(truncated)"
    WeightMinimum=0
    Weight=@n17.OutputParam
   )
   > &RigUnit_AimConstraintLocalSpaceOffset(
    Child=(Type=Bone,Name="base_jacket_spine_cr")
    Filter=(bX=True,bY=True,bZ=True)
    Parents=((Item=(Type=Bone,Name="pelvis"),Weight=1))
    UpAxis=(X=0,Y=1,Z=0)
    Weight=1
    WorldUp=(Target=(X=0,Y=1,Z=0),Kind=Direction,Space=(Type=Bone,Name="base_jacket_spine_cr"))
   )
   &RigUnit_AimConstraintLocalSpaceOffset(
    Child=(Type=Bone,Name="aim_coat_corr_front_01_lf")
    Filter=(bX=false,bY=True,bZ=True)
    Parents=((Item=(Type=Bone,Name="dyn_coat_front_01_lf"),Weight=1))
    Weight=1
    WorldUp=(Target=(X=0,Y=0,Z=1),Kind=Direction,Space=(Type=Bone,Name="dyn_coat_front_01_lf"))
    ExecutePin=@RigVMFunction_Sequence_2.C
   )
   > &RigUnit_AimConstraintLocalSpaceOffset(
    Child=(Type=Bone,Name="aim_coat_corr_front_01_rt")
    Filter=(bX=false,bY=True,bZ=True)
    Parents=((Item=(Type=Bone,Name="dyn_coat_front_01_rt"),Weight=1))
    Weight=1
    WorldUp=(Target=(X=0,Y=0,Z=1),Kind=Direction,Space=(Type=Bone,Name="dyn_coat_front_01_rt"))
   )
   RigVMFunction_MathBoolToFloat = &RigVMFunction_MathBoolToFloat(Value=RigVMVariableNode() {variable=IsCrouch}.Value)
   n18 = &RigVMFunction_AlphaInterp(InterpSpeedDecreasing=10, InterpSpeedIncreasing=10, Scale=1, bClampResult=true, bMapRange=true, Value=@RigVMFunction_MathBoolToFloat.Result)
   n19 = &RigVMFunction_MathDoubleRemap(
    SourceMaximum=45
    SourceMinimum=89
    TargetMaximum=1
    TargetMinimum=0
    Value=&RigUnit_GetRelativeTransformForItem(Child=(Type=Bone,Name="pelvis"), Parent=(Type=Bone,Name="root"), bChildInitial=False, bParentInitial=False).RelativeTransform.Translation.Z
   )
   Multiply = &RigVMFunction_MathDoubleMul(B=@n18.Result, A=@n19.Result)
   > &RigUnit_ModifyTransforms(
    ItemToModify="((Item=(Type=Bone,Name=\"coat_corr_front_01_lf\"),Transform=(Rotation=(X=0.0651,Y=-0.2936,Z=-0.2064,W=0.9311),Translation=(X=3.3606,Y=-1.1619,Z=2.3531),Scale3D=(X=1,Y=1,Z=1))),(Item=(Type=Bone,Name=\"offset_coat_front_lf\"),Transform=(Rotation=(X=-0.3063,Y=-0. ...(truncated)"
    WeightMinimum=0
    Weight=@Multiply.Result
   )
   > &RigUnit_ModifyTransforms(
    ItemToModify="((Item=(Type=Bone,Name=\"base_coat_back_lf\"),Transform=(Rotation=(X=0,Y=0,Z=-0.1305,W=0.9914),Translation=(X=0,Y=0,Z=0),Scale3D=(X=1,Y=1,Z=1))),(Item=(Type=Bone,Name=\"base_coat_back_side_lf\"),Transform=(Rotation=(X=0.0412,Y=0.2117,Z=0.0871,W=0.9726),Transla ...(truncated)"
    WeightMinimum=0
    Weight=@Multiply.Result
   )
   > &RigUnit_ModifyTransforms(
    ItemToModify="((Item=(Type=Bone,Name=\"base_coat_side_rt\"),Transform=(Rotation=(X=-0.0223,Y=-0.0872,Z=0.2408,W=-0.9664),Translation=(X=0,Y=0,Z=0),Scale3D=(X=1,Y=1,Z=1))),(Item=(Type=Bone,Name=\"base_coat_front_side_rt\"),Transform=(Rotation=(X=0.2484,Y=-0.1367,Z=-0.1144,W= ...(truncated)"
    WeightMinimum=0
    Weight=@Multiply.Result
   )
   > &RigUnit_ModifyTransforms(
    ItemToModify="((Item=(Type=Bone,Name=\"dyn_coat_back_02_lf\"),Transform=(Rotation=(X=0,Y=0,Z=0.1736,W=-0.9848),Translation=(X=0,Y=0,Z=0),Scale3D=(X=1,Y=1,Z=1))),(Item=(Type=Bone,Name=\"dyn_coat_back_02_rt\"),Transform=(Rotation=(X=0,Y=0,Z=-0.1736,W=0.9848),Translation=(X=0, ...(truncated)"
    WeightMinimum=0
    Weight=@Multiply.Result
   )
   n20 = &RigVMFunction_AlphaInterp(
    InterpSpeedDecreasing=10
    InterpSpeedIncreasing=10
    Scale=1
    bClampResult=true
    Value=&RigVMFunction_MathBoolToFloat(Value=RigVMVariableNode() {variable=IsCrouchWalking}.Value).Result
   )
   n21 = &RigVMFunction_MathDoubleMul(
    B=@n20.Result
    A=&RigVMFunction_MathDoubleRemap(SourceMaximum=0.7, SourceMinimum=1, TargetMaximum=1, TargetMinimum=0, Value=@n19.Result).Result
   )
   > &RigUnit_ModifyTransforms(
    ItemToModify="((Item=(Type=Bone,Name=\"base_coat_front_lf\"),Transform=(Rotation=(X=-0.0038,Y=0.0438,Z=-0.0867,W=0.9953),Translation=(X=4.3527,Y=1.2224,Z=4.5438),Scale3D=(X=1,Y=1,Z=1))),(Item=(Type=Bone,Name=\"offset_coat_front_rt\"),Transform=(Rotation=(X=0,Y=0,Z=-0.0436,W ...(truncated)"
    WeightMinimum=0
    Weight=@n21.Result
   )
   > &RigUnit_ModifyTransforms(
    ItemToModify="((Item=(Type=Bone,Name=\"offset_coat_side_lf\"),Transform=(Rotation=(X=-0.2754,Y=-0.0134,Z=0.0509,W=0.9599),Translation=(X=0,Y=0,Z=0),Scale3D=(X=1,Y=1,Z=1))),(Item=(Type=Bone,Name=\"offset_coat_side_rt\"),Transform=(Rotation=(X=0,Y=0,Z=-0.0872,W=0.9962),Transl ...(truncated)"
    WeightMinimum=0
    Weight=@n21.Result
   )
   > Delta Rotation = RigVMFunctionReferenceNode(Child=foot_l, Parent=calf_l, Source Maximum=90)
   > &RigUnit_ModifyTransforms(
    ItemToModify="((Item=(Type=Bone,Name=\"base_foot_l\"),Transform=(Rotation=(X=0,Y=-0.4226,Z=0,W=0.9063),Translation=(X=-0.4664,Y=-3.7026,Z=-4.5893),Scale3D=(X=1,Y=1,Z=1))),(Item=(Type=Bone,Name=\"push_foot_l\"),Transform=(Rotation=(X=0,Y=0,Z=0,W=1),Translation=(X=5.2348,Y=3. ...(truncated)"
    WeightMinimum=0
    Weight=@"Delta Rotation".Delta
   )
   > Delta Rotation_1 = RigVMFunctionReferenceNode(Child=foot_r, Parent=calf_r, Source Maximum=90)
   > &RigUnit_ModifyTransforms(
    ItemToModify="((Item=(Type=Bone,Name=\"base_foot_r\"),Transform=(Rotation=(X=0,Y=0.4226,Z=0,W=0.9063),Translation=(X=0,Y=0,Z=0),Scale3D=(X=1,Y=1,Z=1))),(Item=(Type=Bone,Name=\"push_foot_r\"),Transform=(Rotation=(X=0,Y=0,Z=0,W=1),Translation=(X=4.6537,Y=0.3212,Z=-0.386),Scal ...(truncated)"
    WeightMinimum=0
    Weight=@"Delta Rotation_1".Delta
   )
   RigVMFunction_MathBoolNot = &RigVMFunction_MathBoolNot(Value=RigVMVariableNode() {variable=IsOnAir}.Value)
   &RigUnit_AimConstraintLocalSpaceOffset(
    Child=(Type=Bone,Name="pelvis_cloth_helper")
    Filter=(bX=false,bY=True,bZ=True)
    Parents=((Item=(Type=Bone,Name="root"),Weight=1))
    WorldUp=(Target=(X=0,Y=0,Z=1),Kind=Direction,Space=(Type=Bone,Name="root"))
    ExecutePin=@n2.True
    Weight=&RigVMFunction_MathBoolToFloat(Value=@RigVMFunction_MathBoolNot.Result).Result
   )
   &RigUnit_RotationConstraintLocalSpaceOffset(
    AdvancedSettings=(InterpolationType=Average,RotationOrderForFilter=XZY)
    Child=(Type=Bone,Name="coat_cr")
    ExecutePin=()
    Filter=(bX=true,bY=false,bZ=false)
    Parents=((Item=(Type=Bone,Name="spine_03"),Weight=1))
    Weight=1
   )
  ]}
]
hierarchy [
 root {loc=<0 0 0>, rot=<0 0 0 1>}
 pelvis {parent=root, loc=<0 -0.7071 93.1191>, rot=<0.0127 0.707 -0.0127 -0.707>}
 spine_01 {parent=pelvis, loc=<6.7494 -0.6388 0>, rot=<0 0 0.034 -0.9994>}
 spine_02 {parent=spine_01, loc=<7.2518 -0.035 0>, rot=<0 0 0.0238 -0.9997>}
 spine_03 {parent=spine_02, loc=<6.0104 0 0>, rot=<0 0 0.0149 -0.9999>}
 spine_04 {parent=spine_03, loc=<8.1024 0 0>, rot=<0 0 -0.0132 -0.9999>}
 spine_05 {parent=spine_04, loc=<10.0146 0 0>, rot=<0 0 -0.0015 -1>}
 neck_01 {parent=spine_05, loc=<11.3138 0.1529 0>, rot=<0 0 -0.2003 -0.9797>}
 neck_02 {parent=neck_01, loc=<4.5627 0 0>, rot=<0 0 0.0504 -0.9987>}
 head {parent=neck_02, loc=<5.3914 0 0>, rot=<0 0 0.115 -0.9934>}
 attach_cape {parent=spine_05, loc=<9.1257 -11.2929 -0.0422>, rot=<0.0624 -0.7048 -0.0626 -0.7038>}
 cape {parent=attach_cape, type=Null, loc=<0 0 0>, rot=<0 0 0 1>}
 attach_backpack {parent=spine_05, loc=<-0.0211 -13.023 -0.0029>, rot=<-0.0084 -0.7076 0.0083 -0.7065>}
 Backpack_BR {parent=attach_backpack, type=Null, loc=<0 0 0>, rot=<0 0 0 1>}
 clavicle_l {parent=spine_05, loc=<4.3575 -0.1466 -2.3223>, rot=<0.0465 -0.7289 -0.009 -0.6829>}
 upperarm_l {parent=clavicle_l, loc=<15.5111 0 0>, rot=<0.0226 -0.3234 -0.0088 -0.9459>}
 lowerarm_l {parent=upperarm_l, loc=<26.6514 0 0>, rot=<0.0391 -0.0099 -0.2453 -0.9686>}
 hand_l {parent=lowerarm_l, loc=<22.3491 0 0>, rot=<-0.7081 -0.0507 0.0975 -0.6975>}
 thumb_01_l {parent=hand_l, loc=<1.3118 -0.3191 -2.459>, rot=<0.6546 -0.4042 0.0664 -0.6354>}
 thumb_02_l {parent=thumb_01_l, loc=<5.9336 -0.0448 0.0401>, rot=<-0.0037 0.0213 0.298 -0.9543>}
 thumb_03_l {parent=thumb_02_l, loc=<2.2223 0.3682 -0.0907>, rot=<0.0118 -0.013 0.0066 -0.9998>}
 index_metacarpal_l {parent=hand_l, loc=<1.6535 0.0183 -2.96>, rot=<0.0158 -0.0124 -0.0173 -0.9996>}
 index_01_l {parent=index_metacarpal_l, loc=<8.3008 0.0357 -0.0524>, rot=<-0.0044 0.0485 0.0924 -0.9945>}
 index_02_l {parent=index_01_l, loc=<3.3587 -0.3004 0.0452>, rot=<-0.002 0.0027 0.3225 -0.9466>}
 index_03_l {parent=index_02_l, loc=<2.5578 -0.123 -0.0252>, rot=<-0.002 0.0067 0.2558 -0.9667>}
 middle_metacarpal_l {parent=hand_l, loc=<2.1856 -0.1725 -0.5984>, rot=<0.006 0.0168 -0.0329 -0.9993>}
 middle_01_l {parent=middle_metacarpal_l, loc=<7.3275 -0.2447 0.0129>, rot=<-0.0208 0.0042 0.0858 -0.9961>}
 middle_02_l {parent=middle_01_l, loc=<4.0123 0.1401 0.1175>, rot=<-0.0218 0.0258 0.3512 -0.9357>}
 middle_03_l {parent=middle_02_l, loc=<2.6077 0.0534 0.1373>, rot=<0.0082 0.0188 0.1708 -0.9851>}
 ring_metacarpal_l {parent=hand_l, loc=<2.2074 -0.3625 1.3638>, rot=<-0.0284 0.0722 -0.0458 -0.9959>}
 ring_01_l {parent=ring_metacarpal_l, loc=<7.0295 -0.6197 -0.1446>, rot=<-0.0489 -0.0252 0.1892 -0.9804>}
 ring_02_l {parent=ring_01_l, loc=<3.6174 0.4647 -0.1027>, rot=<0 0 0.2806 -0.9598>}
 ring_03_l {parent=ring_02_l, loc=<2.4152 0.114 0.0199>, rot=<0 0 0.1189 -0.9929>}
 pinky_metacarpal_l {parent=hand_l, loc=<1.7616 -0.9753 3.3814>, rot=<-0.0397 0.0814 -0.069 -0.9935>}
 pinky_01_l {parent=pinky_metacarpal_l, loc=<7.1709 -0.4831 -0.1094>, rot=<-0.0654 -0.0035 0.2242 -0.9723>}
 pinky_02_l {parent=pinky_01_l, loc=<2.8274 0.0893 -0.1043>, rot=<0.0037 0.0011 0.2854 -0.9584>}
 pinky_03_l {parent=pinky_02_l, loc=<2.2046 0.0307 -0.0078>, rot=<-0.0002 0.0098 0.1445 -0.9895>}
 deform_wrist_dn_base_l {parent=hand_l, loc=<0 -2.2342 0>, rot=<0 0 0 -1>}
 deform_wrist_dn_l {parent=deform_wrist_dn_base_l, loc=<0 0 0>, rot=<0 0 0 1>}
 deform_wrist_up_base_l {parent=hand_l, loc=<0 2.2682 0>, rot=<0 0 0 -1>}
 deform_wrist_up_l {parent=deform_wrist_up_base_l, loc=<0 0 0>, rot=<0 0 0 1>}
 lowerarm_twist_02_l {parent=lowerarm_l, loc=<9.6199 0 0>, rot=<0 0 0 -1>}
 lowerarm_twist_01_l {parent=lowerarm_l, loc=<17.8549 0 0>, rot=<0 0 0 -1>}
 deform_elbow_in_base_l {parent=lowerarm_l, loc=<0.2112 5.259 0.8274>, rot=<0 0 0 -1>}
 deform_elbow_in_l {parent=deform_elbow_in_base_l, loc=<0 0 0>, rot=<0 0 0 1>}
 deform_elbow_out_base_l {parent=lowerarm_l, loc=<0 -4 0>, rot=<0 0 0 -1>}
 deform_elbow_out_l {parent=deform_elbow_out_base_l, loc=<0 0 0>, rot=<0 0 0 1>}
 upperarm_twist_01_l {parent=upperarm_l, loc=<0.0041 0 0>, rot=<0 0 0 -1>}
 deform_upperarm_bk_base_l {parent=upperarm_twist_01_l, loc=<1.0813 -7.2193 -1.9453>, rot=<0 0 0 1>}
 deform_upperarm_bk_l {parent=deform_upperarm_bk_base_l, loc=<0 0 0>, rot=<0 0 0 1>}
 deform_upperarm_fr_base_l {parent=upperarm_twist_01_l, loc=<-0.3691 7.0506 -2.765>, rot=<0 0 0 1>}
 deform_upperarm_fr_l {parent=deform_upperarm_fr_base_l, loc=<0 0 0>, rot=<0 0 0 1>}
 deform_upperarm_up_base_l {parent=upperarm_twist_01_l, loc=<-3.2216 -1.1391 3.3311>, rot=<0 0 0 1>}
 deform_upperarm_up_l {parent=deform_upperarm_up_base_l, loc=<0 0 0>, rot=<0 0 0 1>}
 upperarm_twist_02_l {parent=upperarm_l, loc=<13.7941 0 0>, rot=<0 0 0 -1>}
 clavicle_r {parent=spine_05, loc=<4.3575 -0.1466 2.3223>, rot=<0.7289 0.0465 -0.6829 0.009>}
 upperarm_r {parent=clavicle_r, loc=<-15.5111 0 0>, rot=<0.0226 -0.3234 -0.0088 -0.9459>}
 lowerarm_r {parent=upperarm_r, loc=<-26.6514 0 0>, rot=<0.0391 -0.0099 -0.2453 -0.9686>}
 hand_r {parent=lowerarm_r, loc=<-22.3491 0 0>, rot=<-0.7081 -0.0507 0.0975 -0.6975>}
 thumb_01_r {parent=hand_r, loc=<-1.3118 0.3191 2.459>, rot=<0.6546 -0.4042 0.0664 -0.6354>}
 thumb_02_r {parent=thumb_01_r, loc=<-5.9336 0.0448 -0.0401>, rot=<-0.0037 0.0213 0.298 -0.9543>}
 thumb_03_r {parent=thumb_02_r, loc=<-2.2223 -0.3682 0.0907>, rot=<0.0118 -0.013 0.0066 -0.9998>}
 index_metacarpal_r {parent=hand_r, loc=<-1.6535 -0.0183 2.96>, rot=<0.0158 -0.0124 -0.0173 -0.9996>}
 index_01_r {parent=index_metacarpal_r, loc=<-8.3008 -0.0357 0.0524>, rot=<-0.0044 0.0485 0.0924 -0.9945>}
 index_02_r {parent=index_01_r, loc=<-3.3587 0.3004 -0.0452>, rot=<-0.002 0.0027 0.3225 -0.9466>}
 index_03_r {parent=index_02_r, loc=<-2.5578 0.123 0.0252>, rot=<-0.002 0.0067 0.2558 -0.9667>}
 middle_metacarpal_r {parent=hand_r, loc=<-2.1856 0.1725 0.5984>, rot=<0.006 0.0168 -0.0329 -0.9993>}
 middle_01_r {parent=middle_metacarpal_r, loc=<-7.3275 0.2447 -0.0129>, rot=<-0.0208 0.0042 0.0858 -0.9961>}
 middle_02_r {parent=middle_01_r, loc=<-4.0123 -0.1401 -0.1175>, rot=<-0.0218 0.0258 0.3512 -0.9357>}
 middle_03_r {parent=middle_02_r, loc=<-2.6077 -0.0534 -0.1373>, rot=<0.0082 0.0188 0.1708 -0.9851>}
 ring_metacarpal_r {parent=hand_r, loc=<-2.2074 0.3625 -1.3638>, rot=<-0.0284 0.0722 -0.0458 -0.9959>}
 ring_01_r {parent=ring_metacarpal_r, loc=<-7.0295 0.6197 0.1446>, rot=<-0.0489 -0.0252 0.1892 -0.9804>}
 ring_02_r {parent=ring_01_r, loc=<-3.6174 -0.4647 0.1027>, rot=<0 0 0.2806 -0.9598>}
 ring_03_r {parent=ring_02_r, loc=<-2.4152 -0.114 -0.0199>, rot=<0 0 0.1189 -0.9929>}
 pinky_metacarpal_r {parent=hand_r, loc=<-1.7616 0.9753 -3.3814>, rot=<-0.0397 0.0814 -0.069 -0.9935>}
 pinky_01_r {parent=pinky_metacarpal_r, loc=<-7.1709 0.4831 0.1094>, rot=<-0.0654 -0.0035 0.2242 -0.9723>}
 pinky_02_r {parent=pinky_01_r, loc=<-2.8274 -0.0893 0.1043>, rot=<0.0037 0.0011 0.2854 -0.9584>}
 pinky_03_r {parent=pinky_02_r, loc=<-2.2046 -0.0307 0.0078>, rot=<-0.0002 0.0098 0.1445 -0.9895>}
 deform_wrist_dn_base_r {parent=hand_r, loc=<-0.0004 2.2345 -0.0001>, rot=<0 0 0 -1>}
 deform_wrist_dn_r {parent=deform_wrist_dn_base_r, loc=<0 0 0>, rot=<0 0 0 1>}
 deform_wrist_up_base_r {parent=hand_r, loc=<0.0003 -2.2684 0.0001>, rot=<0 0 0 -1>}
 deform_wrist_up_r {parent=deform_wrist_up_base_r, loc=<0 0 0>, rot=<0 0 0 1>}
 lowerarm_twist_02_r {parent=lowerarm_r, loc=<-9.6199 0 0>, rot=<0 0 0 -1>}
 lowerarm_twist_01_r {parent=lowerarm_r, loc=<-17.8549 0 0>, rot=<0 0 0 -1>}
 deform_elbow_in_base_r {parent=lowerarm_r, loc=<-0.2111 -5.2591 -0.8275>, rot=<0 0 0 -1>}
 deform_elbow_in_r {parent=deform_elbow_in_base_r, loc=<-0.0001 0 0.0002>, rot=<0 0 0 1>}
 deform_elbow_out_base_r {parent=lowerarm_r, loc=<0 4 -0.0001>, rot=<0 0 0 -1>}
 deform_elbow_out_r {parent=deform_elbow_out_base_r, loc=<0 0 0>, rot=<0 0 0 1>}
 upperarm_twist_01_r {parent=upperarm_r, loc=<-0.0041 0 0>, rot=<0 0 0 -1>}
 deform_upperarm_bk_base_r {parent=upperarm_twist_01_r, loc=<-1.0814 7.2192 1.9455>, rot=<0 0 0 1>}
 deform_upperarm_bk_r {parent=deform_upperarm_bk_base_r, loc=<0 0 0>, rot=<0 0 0 1>}
 deform_upperarm_fr_base_r {parent=upperarm_twist_01_r, loc=<0.3689 -7.0506 2.7652>, rot=<0 0 0 1>}
 deform_upperarm_fr_r {parent=deform_upperarm_fr_base_r, loc=<0 0 0>, rot=<0 0 0 1>}
 deform_upperarm_up_base_r {parent=upperarm_twist_01_r, loc=<3.2214 1.1391 -3.3308>, rot=<0 0 0 1>}
 deform_upperarm_up_r {parent=deform_upperarm_up_base_r, loc=<0 0 0>, rot=<0 0 0 -1>}
 upperarm_twist_02_r {parent=upperarm_r, loc=<-13.7941 0 0>, rot=<0 0 0 -1>}
 base_hood_cr {parent=spine_05, loc=<11.2756 0.1721 0>, rot=<0.5196 -0.4795 0.4795 -0.5196>}
 dyn_hood_cr {parent=base_hood_cr, loc=<0 0 0>, rot=<0 0 0 1>}
 deform_pec_aim_base_l {parent=spine_04, loc=<7.8732 12.6388 -2.4123>, rot=<0.0102 -0.4602 0.3979 -0.7936>}
 deform_pec_aim_l {parent=deform_pec_aim_base_l, loc=<0 0 0>, rot=<0 0 0 1>}
 deform_pec_base_l {parent=deform_pec_aim_l, loc=<10.766 0 0>, rot=<0 0 0 1>}
 deform_pec_l {parent=deform_pec_base_l, loc=<0.055 0 0>, rot=<0 0 0 1>}
 deform_pec_aim_base_r {parent=spine_04, loc=<7.8727 12.6388 2.4123>, rot=<-0.4602 -0.0102 0.7936 0.3979>}
 deform_pec_aim_r {parent=deform_pec_aim_base_r, loc=<0 0 0>, rot=<0 0 0 1>}
 deform_pec_base_r {parent=deform_pec_aim_r, loc=<-10.7662 -0.0006 -0.0007>, rot=<0 0 0 1>}
 deform_pec_r {parent=deform_pec_base_r, loc=<0 0 0>, rot=<0 0 0 1>}
 base_belt_front_lf {parent=spine_04, loc=<3.3432 10.2591 -12.4225>, rot=<0.0294 -0.7065 -0.0294 -0.7065>}
 dyn_belt_front_lf {parent=base_belt_front_lf, loc=<0 0 0>, rot=<0 0 0 1>}
 base_belt_front_rt {parent=spine_04, loc=<3.3429 10.2591 12.4225>, rot=<0.7065 0.0294 0.7065 -0.0294>}
 dyn_belt_front_rt {parent=base_belt_front_rt, loc=<0 0 0>, rot=<0 0 0 1>}
 base_jacket_spine_cr {parent=spine_03, loc=<0 0 0>, rot=<0 0 0 1>}
 jacket_spine_cr {parent=base_jacket_spine_cr, loc=<0 0 0>, rot=<0 0 0 1>}
 base_coat_corr_front_lf {parent=jacket_spine_cr, loc=<0.3654 12.4701 -7.5828>, rot=<-0.0361 0.1437 0.9887 -0.0216>}
 aim_coat_corr_front_01_lf {parent=base_coat_corr_front_lf, loc=<0 0 0>, rot=<0 0 0 1>}
 coat_corr_front_01_lf {parent=aim_coat_corr_front_01_lf, loc=<0 0 0>, rot=<0 0 0 1>}
 base_coat_corr_front_rt {parent=jacket_spine_cr, loc=<0.3659 12.4701 7.5828>, rot=<0.0361 -0.1437 0.9887 -0.0216>}
 aim_coat_corr_front_01_rt {parent=base_coat_corr_front_rt, loc=<0 0 0>, rot=<0 0 0 1>}
 coat_corr_front_01_rt {parent=aim_coat_corr_front_01_rt, loc=<0 0 0>, rot=<0 0 0 1>}
 base_belt_back_lf {parent=spine_03, loc=<8.8588 -9.9831 0>, rot=<0.5266 -0.4719 -0.5266 -0.4719>}
 dyn_belt_back_lf {parent=base_belt_back_lf, loc=<0 0 0>, rot=<0 0 0 1>}
 base_belt_back_rt {parent=spine_03, loc=<8.8591 -9.9831 0>, rot=<-0.4626 -0.5348 0.481 -0.5183>}
 dyn_belt_back_rt {parent=base_belt_back_rt, loc=<0 0 0>, rot=<0 0 1 0>}
 thigh_l {parent=pelvis, loc=<-2.5693 -0.0109 -7.7501>, rot=<0.0012 0.0572 0.0155 -0.9982>}
 calf_l {parent=thigh_l, loc=<-39.0484 0 0>, rot=<0.0006 -0.0161 -0.0391 -0.9991>}
 foot_l {parent=calf_l, loc=<-42.561 0.1077 0.4292>, rot=<0.0418 -0.0455 0.0343 -0.9975>}
 ball_l {parent=foot_l, loc=<-6.1599 13.8264 0.6806>, rot=<0.0001 -0.0001 -0.7063 -0.7079>}
 base_foot_l {parent=foot_l, loc=<0 0 0>, rot=<-0.4787 0.4787 0.5204 -0.5204>}
 push_foot_l {parent=base_foot_l, loc=<0 0 0>, rot=<0 0 0 1>}
 calf_twist_02_l {parent=calf_l, loc=<-18.8064 -0.0967 0.0397>, rot=<0.0237 0.0095 -0.0035 -0.9997>}
 calf_twist_01_l {parent=calf_l, loc=<-35.813 -0.0967 0.0397>, rot=<0.0237 0.0095 -0.0035 -0.9997>}
 deform_knee_in_base_l {parent=calf_l, loc=<1.298 -5.2957 -0.0496>, rot=<0 0 0 -1>}
 deform_knee_in_l {parent=deform_knee_in_base_l, loc=<0 0 0>, rot=<0 0 0 1>}
 deform_knee_out_base_l {parent=calf_l, loc=<0 4 0>, rot=<0 0 0 -1>}
 deform_knee_out_l {parent=deform_knee_out_base_l, loc=<0 0 0>, rot=<0 0 0 1>}
 thigh_twist_01_l {parent=thigh_l, loc=<-21.6303 0 0>, rot=<0.0013 0 0 -1>}
 deform_groin_base_l {parent=thigh_l, loc=<-2.1908 8.1142 -1.7442>, rot=<0 0 0 -1>}
 deform_groin_l {parent=deform_groin_base_l, loc=<0 0 0>, rot=<0 0 0 1>}
 thigh_r {parent=pelvis, loc=<-2.5693 -0.0109 7.7501>, rot=<0.0572 -0.0012 0.9982 0.0155>}
 calf_r {parent=thigh_r, loc=<39.0484 0 0>, rot=<0.0006 -0.0161 -0.0391 -0.9991>}
 foot_r {parent=calf_r, loc=<42.561 -0.1077 -0.4292>, rot=<0.0418 -0.0455 0.0343 -0.9975>}
 ball_r {parent=foot_r, loc=<6.1599 -13.8264 -0.6806>, rot=<0.0001 -0.0001 -0.7063 -0.7079>}
 base_foot_r {parent=foot_r, loc=<0 0 0>, rot=<0.5074 0.4925 -0.4926 -0.5072>}
 push_foot_r {parent=base_foot_r, loc=<0.1573 0.0116 -0.0134>, rot=<0 0 0 1>}
 calf_twist_02_r {parent=calf_r, loc=<18.8064 0.0967 -0.0397>, rot=<0.0237 0.0095 -0.0035 -0.9997>}
 calf_twist_01_r {parent=calf_r, loc=<35.813 0.0967 -0.0397>, rot=<0.0237 0.0095 -0.0035 -0.9997>}
 deform_knee_in_base_r {parent=calf_r, loc=<-1.298 5.2957 0.0496>, rot=<0 0 0 -1>}
 deform_knee_in_r {parent=deform_knee_in_base_r, loc=<0 0 0>, rot=<0 0 0 1>}
 deform_knee_out_base_r {parent=calf_r, loc=<0 -4 0>, rot=<0 0 0 -1>}
 deform_knee_out_r {parent=deform_knee_out_base_r, loc=<0 0 0>, rot=<0 0 0 1>}
 thigh_twist_01_r {parent=thigh_r, loc=<21.6303 0 0>, rot=<0.0013 0 0 -1>}
 deform_groin_base_r {parent=thigh_r, loc=<2.1908 -8.1142 1.7442>, rot=<0 0 0 -1>}
 deform_groin_r {parent=deform_groin_base_r, loc=<0 0 0>, rot=<0 0 0 1>}
 deform_glute_base_l {parent=pelvis, loc=<-8.8064 -8.8222 -8.4287>, rot=<0.0275 0.0501 -0.4472 -0.8926>}
 deform_glute_l {parent=deform_glute_base_l, loc=<0 0 0>, rot=<0 0 0 1>}
 deform_glute_base_r {parent=pelvis, loc=<-8.8062 -8.8222 8.4287>, rot=<0.0501 -0.0275 0.8926 -0.4472>}
 deform_glute_r {parent=deform_glute_base_r, loc=<0 0 0>, rot=<0 0 0 1>}
 base_coat_cr {parent=pelvis, loc=<0 0 0>, rot=<0 0 0 1>}
 coat_cr {parent=base_coat_cr, loc=<0 0 0>, rot=<0 0 0 1>}
 base_coat_back_side_lf {parent=coat_cr, loc=<12.2124 -5.9447 -9.4161>, rot=<0.057 -0.9706 -0.2318 -0.0307>}
 offset_coat_back_side_lf {parent=base_coat_back_side_lf, loc=<0 0 0>, rot=<0 0 0 1>}
 dyn_coat_back_side_01_lf {parent=offset_coat_back_side_lf, loc=<0 0 0>, rot=<0 0 0 1>}
 dyn_coat_back_side_02_lf {parent=dyn_coat_back_side_01_lf, loc=<18.6879 0 0>, rot=<0 0 0 1>}
 dyn_coat_back_side_03_lf {parent=dyn_coat_back_side_02_lf, loc=<18.6879 0 0>, rot=<0 0 0 1>}
 dyn_coat_back_side_04_lf {parent=dyn_coat_back_side_03_lf, loc=<18.6879 0 0>, rot=<0 0 0 1>}
 base_coat_side_lf {parent=coat_cr, loc=<11.9769 -1.1996 -12.9347>, rot=<0.0875 -0.8314 -0.5443 -0.0694>}
 offset_coat_side_lf {parent=base_coat_side_lf, loc=<0 0 0>, rot=<0 0 0 1>}
 dyn_coat_side_01_lf {parent=offset_coat_side_lf, loc=<0 0 0>, rot=<0 0 0 1>}
 dyn_coat_side_02_lf {parent=dyn_coat_side_01_lf, loc=<18.688 0 0>, rot=<0 0 0 1>}
 dyn_coat_side_03_lf {parent=dyn_coat_side_02_lf, loc=<18.688 0 0>, rot=<0 0 0 1>}
 dyn_coat_side_04_lf {parent=dyn_coat_side_03_lf, loc=<18.688 0 0>, rot=<0 0 0 1>}
 base_coat_front_side_lf {parent=coat_cr, loc=<12.24 6.2902 -12.9583>, rot=<-0.0724 0.5569 0.8243 0.0722>}
 offset_coat_front_side_lf {parent=base_coat_front_side_lf, loc=<0 0 0>, rot=<0 0 0 1>}
 dyn_coat_front_side_01_lf {parent=offset_coat_front_side_lf, loc=<0 0 0>, rot=<0 0 0 1>}
 dyn_coat_front_side_02_lf {parent=dyn_coat_front_side_01_lf, loc=<18.688 0 0>, rot=<0 0 0 1>}
 dyn_coat_front_side_03_lf {parent=dyn_coat_front_side_02_lf, loc=<18.688 0 0>, rot=<0 0 0 1>}
 dyn_coat_front_side_04_lf {parent=dyn_coat_front_side_03_lf, loc=<18.688 0 0>, rot=<0 0 0 1>}
 base_coat_back_rt {parent=coat_cr, loc=<11.985 -8.4732 4.4925>, rot=<0.0517 -0.9982 0.0296 0.0108>}
 offset_coat_back_rt {parent=base_coat_back_rt, loc=<0 0 0>, rot=<0 0 0 1>}
 dyn_coat_back_01_rt {parent=offset_coat_back_rt, loc=<0 0 0>, rot=<0 0 0 1>}
 dyn_coat_back_02_rt {parent=dyn_coat_back_01_rt, loc=<18.6878 -0.0001 0>, rot=<0 0 0 1>}
 dyn_coat_back_03_rt {parent=dyn_coat_back_02_rt, loc=<18.688 0.0001 0>, rot=<0 0 0 1>}
 dyn_coat_back_04_rt {parent=dyn_coat_back_03_rt, loc=<18.688 0 0>, rot=<0 0 0 1>}
 base_coat_side_rt {parent=coat_cr, loc=<11.9766 -1.1996 12.9347>, rot=<-0.8314 -0.0875 0.0694 -0.5443>}
 offset_coat_side_rt {parent=base_coat_side_rt, loc=<0 0 0>, rot=<0 0 1 0>}
 dyn_coat_side_01_rt {parent=offset_coat_side_rt, loc=<0 0 0>, rot=<0 0 0 1>}
 dyn_coat_side_02_rt {parent=dyn_coat_side_01_rt, loc=<18.6876 -0.0001 0>, rot=<0 0 0 1>}
 dyn_coat_side_03_rt {parent=dyn_coat_side_02_rt, loc=<18.688 0 0>, rot=<0 0 0 1>}
 dyn_coat_side_04_rt {parent=dyn_coat_side_03_rt, loc=<18.688 0 0>, rot=<0 0 0 1>}
 base_coat_front_side_rt {parent=coat_cr, loc=<12.2401 6.2902 12.9583>, rot=<0.0724 -0.5569 0.8243 0.0722>}
 offset_coat_front_side_rt {parent=base_coat_front_side_rt, loc=<0 0 0>, rot=<0 0 0 1>}
 dyn_coat_front_side_01_rt {parent=offset_coat_front_side_rt, loc=<0 0 0>, rot=<0 0 0 1>}
 dyn_coat_front_side_02_rt {parent=dyn_coat_front_side_01_rt, loc=<18.688 0 0>, rot=<0 0 0 1>}
 dyn_coat_front_side_03_rt {parent=dyn_coat_front_side_02_rt, loc=<18.688 0.0001 0>, rot=<0 0 0 1>}
 dyn_coat_front_side_04_rt {parent=dyn_coat_front_side_03_rt, loc=<18.688 0 0>, rot=<0 0 0 1>}
 base_coat_back_side_rt {parent=coat_cr, loc=<12.2128 -5.9447 9.4161>, rot=<0.057 -0.9706 0.2318 0.0307>}
 offset_coat_back_side_rt {parent=base_coat_back_side_rt, loc=<0 0 0>, rot=<0 0 0 1>}
 dyn_coat_back_side_01_rt {parent=offset_coat_back_side_rt, loc=<0 0 0>, rot=<0 0 0 1>}
 dyn_coat_back_side_02_rt {parent=dyn_coat_back_side_01_rt, loc=<18.6883 0.0001 0>, rot=<0 0 0 1>}
 dyn_coat_back_side_03_rt {parent=dyn_coat_back_side_02_rt, loc=<18.6879 0 0>, rot=<0 0 0 1>}
 dyn_coat_back_side_04_rt {parent=dyn_coat_back_side_03_rt, loc=<18.6881 0.0001 0>, rot=<0 0 0 1>}
 base_coat_back_lf {parent=coat_cr, loc=<11.9852 -8.4733 -4.4925>, rot=<0.0517 -0.9982 -0.0296 -0.0108>}
 offset_coat_back_lf {parent=base_coat_back_lf, loc=<0 0 0>, rot=<0 0 0 1>}
 dyn_coat_back_01_lf {parent=offset_coat_back_lf, loc=<0 0 0>, rot=<0 0 0 1>}
 dyn_coat_back_02_lf {parent=dyn_coat_back_01_lf, loc=<18.688 0 0>, rot=<0 0 0 1>}
 dyn_coat_back_03_lf {parent=dyn_coat_back_02_lf, loc=<18.688 0 0>, rot=<0 0 0 1>}
 dyn_coat_back_04_lf {parent=dyn_coat_back_03_lf, loc=<18.688 0 0>, rot=<0 0 0 1>}
 orient_coat_front_lf {parent=coat_cr, loc=<12.6214 11.303 -8.2061>, rot=<0 0 -0.7071 -0.7071>}
 base_coat_front_lf {parent=orient_coat_front_lf, loc=<0 0 0>, rot=<-0.0867 -0.1596 -0.6609 -0.7282>}
 offset_coat_front_lf {parent=base_coat_front_lf, loc=<0 0 0>, rot=<0 0 0 1>}
 dyn_coat_front_01_lf {parent=offset_coat_front_lf, loc=<0 0 0>, rot=<0 0 0 1>}
 dyn_coat_front_02_lf {parent=dyn_coat_front_01_lf, loc=<8.0861 0 0>, rot=<0 0 -0.0199 -0.9998>}
 dyn_coat_front_03_lf {parent=dyn_coat_front_02_lf, loc=<10.105 -0.4022 0>, rot=<0 0 0 1>}
 dyn_coat_front_04_lf {parent=dyn_coat_front_03_lf, loc=<18.199 0 0>, rot=<0 0 0 1>}
 dyn_coat_front_05_lf {parent=dyn_coat_front_04_lf, loc=<18.199 0 0>, rot=<0 0 0 1>}
 orient_coat_front_rt {parent=coat_cr, loc=<12.6219 11.3029 8.2061>, rot=<0 0 -0.7071 -0.7071>}
 base_coat_front_rt {parent=orient_coat_front_rt, loc=<0 0 0>, rot=<0.0867 0.1596 -0.6609 -0.7282>}
 offset_coat_front_rt {parent=base_coat_front_rt, loc=<0 0 0>, rot=<0 0 0 1>}
 dyn_coat_front_01_rt {parent=offset_coat_front_rt, loc=<0 0 0>, rot=<0 0 0 1>}
 dyn_coat_front_02_rt {parent=dyn_coat_front_01_rt, loc=<8.086 0 0>, rot=<0 0 -0.0199 -0.9998>}
 dyn_coat_front_03_rt {parent=dyn_coat_front_02_rt, loc=<10.1055 -0.4023 0>, rot=<0 0 0 1>}
 dyn_coat_front_04_rt {parent=dyn_coat_front_03_rt, loc=<18.199 0 0>, rot=<0 0 0 1>}
 dyn_coat_front_05_rt {parent=dyn_coat_front_04_rt, loc=<18.199 0 0>, rot=<0 0 0 1>}
 pelvis_cloth_helper {parent=pelvis, loc=<0 0 0>, rot=<0 0 0 1>}
 cloth_groung_collision {parent=root, loc=<0 0 0>, rot=<0 0 0 1>}
 cloth_collision_helper {parent=root, loc=<0 0 0>, rot=<0 0 0 1>}
 physic_ground {parent=root, loc=<0 0 0>, rot=<0 0 0 1>}
]
variables [
 IsCloth {type=bool, default=False}
 IsCrouch {type=bool, default=False}
 IsCrouchWalking {type=bool, default=False}
 IsJumping {type=bool, default=False}
 IsOnAir {type=bool, default=False}
]
