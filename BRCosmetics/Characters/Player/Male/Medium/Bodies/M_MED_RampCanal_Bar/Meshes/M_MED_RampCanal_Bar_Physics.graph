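# Physics asset for the RampCanal_Bar male medium player body, simulating its long coat against the legs and ground.
# Five capsule chains per side (front, front-side, side, back-side, back) hang from small offset anchor bodies, with narrow swing limits that generally widen downward and mostly locked twist.
# Pelvis, thigh and calf capsules plus a huge physic_ground capsule provide collision shapes.
# Left front and front-side chains lack constraints to their anchors; left front also lacks its 01-to-02 link.
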PhysicsAsset /BRCosmetics/Characters/Player/Male/Medium/Bodies/M_MED_RampCanal_Bar/Meshes/M_MED_RampCanal_Bar_Physics
bodies {
 calf_l [{shapeType=Capsule, length=43.1981, radius=6.2867, center=<-29.5676 -3.3979 -0.2637>, rotation=<rot -0.0461 87.6516 -89.7388>}]
 calf_r [{shapeType=Capsule, length=43.1981, radius=6.2867, center=<29.5119 2.04 0.2648>, rotation=<rot -0.0461 87.6516 90.2612>}]
 dyn_coat_back_01_lf [{shapeType=Capsule, length=10, radius=5, shapeName=dyn_coat_back_01_rt_capsule, center=<14.5 -5.6 0>, rotation=<rot 90 0 0>}]
 dyn_coat_back_01_rt [{shapeType=Capsule, length=10, radius=5, shapeName=dyn_coat_back_01_rt_capsule, center=<14.5 -5.6 0>, rotation=<rot 90 0 0>}]
 dyn_coat_back_02_lf [{shapeType=Capsule, length=13, radius=10, shapeName=dyn_coat_back_02_rt_capsule, center=<11.9507 -5.5 6>}]
 dyn_coat_back_02_rt [{shapeType=Capsule, length=13, radius=10, shapeName=dyn_coat_back_02_rt_capsule, center=<11.9507 -5.5 -6>}]
 dyn_coat_back_03_lf [{shapeType=Capsule, length=7.5, radius=11, shapeName=dyn_coat_back_03_lf_capsule, center=<12 -4.6093 7.9528>, rotation=<rot 0 0 9.2421>}]
 dyn_coat_back_03_rt [{shapeType=Capsule, length=8, radius=10.5, shapeName=dyn_coat_back_03_lf_capsule, center=<12 -4.6093 -7.9528>, rotation=<rot 0 0 -9.2421>}]
 dyn_coat_back_04_lf [{
   shapeType=Capsule
   length=19
   radius=6
   shapeName=dyn_coat_back_04_lf_capsule
   center=<17.9553 -0.5169 8.7992>
   rotation=<rot 0 0.0002 5.9819>
  }]
 dyn_coat_back_04_rt [{
   shapeType=Capsule
   length=18.5
   radius=6.5
   shapeName=dyn_coat_back_04_lf_capsule
   center=<18.9477 -0.9462 -8.6454>
   rotation=<rot 0 0.0002 -5.9819>
  }]
 dyn_coat_back_side_01_lf [{shapeType=Capsule, length=11, radius=4, shapeName=dyn_coat_back_side_01_rt_capsule, center=<14.3612 -4.0292 0>, rotation=<rot -90 0.0001 0>}]
 dyn_coat_back_side_01_rt [{shapeType=Capsule, length=11, radius=4, shapeName=dyn_coat_back_side_01_rt_capsule, center=<14.3612 -4.7097 0>, rotation=<rot -90 0.0001 0>}]
 dyn_coat_back_side_02_lf [{shapeType=Capsule, length=20.5, radius=7.5, shapeName=dyn_coat_back_side_02_rt_capsule, center=<10 -5 0>}]
 dyn_coat_back_side_02_rt [{shapeType=Capsule, length=21, radius=7, shapeName=dyn_coat_back_side_02_rt_capsule, center=<10 -5 0>}]
 dyn_coat_back_side_03_lf [{shapeType=Capsule, length=21.5, radius=10, shapeName=dyn_coat_back_side_03_lf_capsule, center=<10 -5 0>}]
 dyn_coat_back_side_03_rt [{shapeType=Capsule, length=21.5, radius=10, shapeName=dyn_coat_back_side_03_lf_capsule, center=<10 -5 0>}]
 dyn_coat_back_side_04_lf [{
   shapeType=Capsule
   length=24
   radius=5.5
   shapeName=dyn_coat_back_side_04_lf_capsule
   center=<14.5408 -4.3426 0.3284>
   rotation=<rot 0 0.0001 -10>
  }]
 dyn_coat_back_side_04_rt [{shapeType=Capsule, length=35, radius=5.5, shapeName=dyn_coat_back_side_04_lf_capsule, center=<10 -5.2698 -3.0838>, rotation=<rot 0 0.0001 0>}]
 dyn_coat_front_01_lf [{shapeType=Capsule, length=11.5, radius=4.5, shapeName=dyn_coat_front_01_rt_capsule, center=<5 -2 0>}]
 dyn_coat_front_01_rt [{shapeType=Capsule, length=11.5, radius=4.5, shapeName=dyn_coat_front_01_rt_capsule, center=<5 -2 0>}]
 dyn_coat_front_02_lf [{shapeType=Capsule, length=13, radius=7, shapeName=dyn_coat_front_02_lf_capsule, center=<10.6552 -3.1248 -7.1341>, rotation=<rot -90 0 0>}]
 dyn_coat_front_02_rt [{shapeType=Capsule, length=11, radius=9, shapeName=dyn_coat_front_02_rt_capsule, center=<10 -4.7396 10.55>, rotation=<rot 0 -0.0001 30>}]
 dyn_coat_front_03_lf [{
   shapeType=Capsule
   length=18.3932
   radius=9.1068
   shapeName=dyn_coat_front_05_lf_capsule
   center=<18.345 -3.0785 -11.1622>
   rotation=<rot 0.0001 -0.0003 -25>
  }]
 dyn_coat_front_03_rt [{shapeType=Capsule, length=11, radius=9, shapeName=dyn_coat_front_03_lf_capsule, center=<10 -4.2566 8.993>, rotation=<rot 0 -0.0001 40>}]
 dyn_coat_front_04_lf [{
   shapeType=Capsule
   length=18
   radius=9.5
   shapeName=dyn_coat_front_05_lf_capsule
   center=<18.345 -3.0785 -11.1622>
   rotation=<rot 0.0001 -0.0003 -30>
  }]
 dyn_coat_front_04_rt [{
   shapeType=Capsule
   length=13.5
   radius=9
   shapeName=dyn_coat_front_03_lf_capsule
   center=<10 -2.2757 11.6755>
   rotation=<rot -0.0001 -0.0002 30>
  }]
 dyn_coat_front_05_lf [{
   shapeType=Capsule
   length=14.5
   radius=10
   shapeName=dyn_coat_front_05_lf_capsule
   center=<28.5127 -1.7056 -11.74>
   rotation=<rot 0 -0.0002 150>
  }]
 dyn_coat_front_05_rt [{
   shapeType=Capsule
   length=14.5
   radius=10
   shapeName=dyn_coat_front_05_lf_capsule
   center=<21.8142 -1.7056 11.74>
   rotation=<rot -0.0001 -0.0001 45>
  }]
 dyn_coat_front_side_01_lf [{shapeType=Capsule, length=10, radius=3, shapeName=dyn_coat_front_side_01_lf_capsule, center=<4 0 0>, rotation=<rot 90 -15 0>}]
 dyn_coat_front_side_01_rt [{
   shapeType=Capsule
   length=13.4944
   radius=8.0582
   shapeName=dyn_coat_front_side_01_rt_capsule
   center=<12.5514 -7.0809 -1.5992>
   rotation=<rot 0 -87.7375 -86.4606>
  }]
 dyn_coat_front_side_02_lf [{shapeType=Capsule, length=13.5, radius=9, shapeName=dyn_coat_front_side_02_lf_capsule, center=<10 -5.8028 0>}]
 dyn_coat_front_side_02_rt [{
   shapeType=Capsule
   length=13.5
   radius=9
   shapeName=dyn_coat_front_side_02_lf_capsule
   center=<10 -8 -0.0798>
   rotation=<rot -0.0003 -0.0001 -4.9999>
  }]
 dyn_coat_front_side_03_lf [{shapeType=Capsule, length=13.5, radius=9, shapeName=dyn_coat_front_side_03_lf_capsule, center=<10 -5.8028 0>}]
 dyn_coat_front_side_03_rt [{shapeType=Capsule, length=21.9541, radius=7, shapeName=dyn_coat_front_side_03_lf_capsule, center=<10 -5.5295 1.6687>}]
 dyn_coat_front_side_04_lf [
  {shapeType=Capsule, length=13.5, radius=9, shapeName=dyn_coat_front_side_04_rt_capsule, center=<13.2722 -5.8028 0>}
  {shapeType=Capsule, length=16.5, radius=4.5, shapeName=dyn_coat_front_side_04_lf_capsule, center=<11.3558 -0.9358 0>, rotation=<rot -90 0 0>}
 ]
 dyn_coat_front_side_04_rt [
  {shapeType=Capsule, length=18.5, radius=9, shapeName=dyn_coat_front_side_04_rt_capsule, center=<11.4286 -3.179 0>}
  {shapeType=Capsule, length=9.5, radius=5, shapeName=dyn_coat_front_side_04_rt_capsule, center=<19.281 0 0>, rotation=<rot 90 -0.0001 0>}
 ]
 dyn_coat_side_01_lf [{shapeType=Capsule, length=8.1666, radius=4.6815, shapeName=dyn_coat_side_01_lf_capsule}]
 dyn_coat_side_01_rt [{shapeType=Capsule, length=8.1682, radius=5.2083, shapeName=dyn_coat_side_01_rt_capsule, center=<8.0578 0 0>}]
 dyn_coat_side_02_lf [{shapeType=Capsule, length=17, radius=4.5, shapeName=dyn_coat_side_02_lf_capsule, center=<10 -2.6086 0>}]
 dyn_coat_side_02_rt [{shapeType=Capsule, length=16, radius=5.5, shapeName=dyn_coat_side_02_lf_capsule, center=<10 -2.6086 0>}]
 dyn_coat_side_03_lf [{shapeType=Capsule, length=15.5, radius=6, shapeName=dyn_coat_side_03_lf_capsule, center=<10 0 0>}]
 dyn_coat_side_03_rt [{shapeType=Capsule, length=15.5, radius=6, shapeName=dyn_coat_side_03_lf_capsule, center=<10 0 0>}]
 dyn_coat_side_04_lf [{shapeType=Capsule, length=16.5, radius=5, shapeName=dyn_coat_side_04_lf_capsule, center=<8.1978 -2.6384 0>, rotation=<rot 90 0.0001 0>}]
 dyn_coat_side_04_rt [{shapeType=Capsule, length=15, radius=4, shapeName=dyn_coat_side_04_lf_capsule, center=<4.2507 0 0>, rotation=<rot -74.9999 -90.0007 90.0009>}]
 offset_coat_back_lf [{shapeType=Capsule, length=0.505, radius=0.505, shapeName=offset_coat_back_lf_capsule, rotation=<rot 0 -90 0>}]
 offset_coat_back_rt [{shapeType=Capsule, length=0.505, radius=0.505, shapeName=offset_coat_back_rt_capsule, rotation=<rot 0 -90 0>}]
 offset_coat_back_side_lf [{shapeType=Capsule, length=0.505, radius=0.505, shapeName=offset_coat_back_side_lf_capsule, rotation=<rot 0 -90 0>}]
 offset_coat_back_side_rt [{shapeType=Capsule, length=0.505, radius=0.505, shapeName=offset_coat_back_side_rt_capsule, rotation=<rot 0 -90 0>}]
 offset_coat_front_lf [{shapeType=Capsule, length=0.505, radius=0.505, shapeName=offset_coat_front_lf_capsule, rotation=<rot 0 -90 0>}]
 offset_coat_front_rt [{shapeType=Capsule, length=0.505, radius=0.505, shapeName=offset_coat_front_rt_capsule, rotation=<rot 0 -90 0>}]
 offset_coat_front_side_lf [{shapeType=Capsule, length=0.505, radius=0.505, shapeName=offset_coat_front_side_lf_capsule, rotation=<rot 0 -90 0>}]
 offset_coat_front_side_rt [{shapeType=Capsule, length=0.505, radius=0.505, shapeName=offset_coat_front_side_rt_capsule, rotation=<rot 0 -90 0>}]
 offset_coat_side_lf [{shapeType=Capsule, length=0.505, radius=0.505, shapeName=offset_coat_side_lf_capsule, rotation=<rot 0 -90 0>}]
 offset_coat_side_rt [{shapeType=Capsule, length=0.505, radius=0.505, shapeName=offset_coat_side_rt_capsule, rotation=<rot 0 -90 0>}]
 pelvis [{shapeType=Capsule, length=12.4626, radius=9.5991, center=<-2.933 0.4838 0>, rotation=<rot 0 -90 0>}]
 physic_ground [{shapeType=Capsule, length=5, radius=502, shapeName=root_capsule, center=<0 0 -505>}]
 thigh_l [{shapeType=Capsule, length=29.2797, radius=7.8206, center=<-23.078 0.9595 0.5463>, rotation=<rot -0.4094 -88.7186 -87.3226>}]
 thigh_r [{shapeType=Capsule, length=30.2797, radius=7.3206, center=<22.2683 -0.6482 0.1861>, rotation=<rot -0.4094 -88.7186 92.6774>}]
}
constraints [
 {
  bone1Name=dyn_coat_back_01_lf
  bone2Name=offset_coat_back_lf
  swing1LimitDegrees=5
  swing2LimitDegrees=0
  twistLimitDegrees=45
  twistMotion=Locked
 }
 {
  bone1Name=dyn_coat_back_02_lf
  bone2Name=dyn_coat_back_01_lf
  swing1LimitDegrees=10
  swing2LimitDegrees=0
  twistLimitDegrees=45
  twistMotion=Locked
 }
 {
  bone1Name=dyn_coat_back_03_lf
  bone2Name=dyn_coat_back_02_lf
  swing1LimitDegrees=15
  swing2LimitDegrees=6
  twistLimitDegrees=45
  twistMotion=Locked
 }
 {
  bone1Name=dyn_coat_back_04_lf
  bone2Name=dyn_coat_back_03_lf
  swing1LimitDegrees=15
  swing2LimitDegrees=5
  twistLimitDegrees=0
  twistMotion=Locked
 }
 {
  bone1Name=dyn_coat_back_01_rt
  bone2Name=offset_coat_back_rt
  swing1LimitDegrees=5
  swing2LimitDegrees=0
  twistLimitDegrees=45
  twistMotion=Locked
 }
 {
  bone1Name=dyn_coat_back_02_rt
  bone2Name=dyn_coat_back_01_rt
  swing1LimitDegrees=10
  swing2LimitDegrees=0
  twistLimitDegrees=45
  twistMotion=Locked
 }
 {
  bone1Name=dyn_coat_back_03_rt
  bone2Name=dyn_coat_back_02_rt
  swing1LimitDegrees=15
  swing2LimitDegrees=5
  twistLimitDegrees=45
  twistMotion=Locked
 }
 {
  bone1Name=dyn_coat_back_04_rt
  bone2Name=dyn_coat_back_03_rt
  swing1LimitDegrees=15
  swing2LimitDegrees=5
  twistLimitDegrees=0
  twistMotion=Locked
 }
 {
  bone1Name=dyn_coat_back_side_01_lf
  bone2Name=offset_coat_back_side_lf
  swing1LimitDegrees=10
  swing2LimitDegrees=0
  twistLimitDegrees=45
  twistMotion=Locked
 }
 {
  bone1Name=dyn_coat_back_side_02_lf
  bone2Name=dyn_coat_back_side_01_lf
  swing1LimitDegrees=15
  swing2LimitDegrees=0
  twistLimitDegrees=45
  twistMotion=Locked
 }
 {
  bone1Name=dyn_coat_back_side_03_lf
  bone2Name=dyn_coat_back_side_02_lf
  swing1LimitDegrees=20
  swing2LimitDegrees=5
  twistLimitDegrees=45
  twistMotion=Locked
 }
 {
  bone1Name=dyn_coat_back_side_04_lf
  bone2Name=dyn_coat_back_side_03_lf
  swing1LimitDegrees=15
  swing2LimitDegrees=5
  twistLimitDegrees=45
  twistMotion=Locked
 }
 {
  bone1Name=dyn_coat_back_side_01_rt
  bone2Name=offset_coat_back_side_rt
  swing1LimitDegrees=5
  swing2LimitDegrees=0
  twistLimitDegrees=45
  twistMotion=Locked
 }
 {
  bone1Name=dyn_coat_back_side_02_rt
  bone2Name=dyn_coat_back_side_01_rt
  swing1LimitDegrees=15
  swing2LimitDegrees=0
  twistLimitDegrees=45
  twistMotion=Locked
 }
 {
  bone1Name=dyn_coat_back_side_03_rt
  bone2Name=dyn_coat_back_side_02_rt
  swing1LimitDegrees=20
  swing2LimitDegrees=0
  twistLimitDegrees=45
  twistMotion=Locked
 }
 {
  bone1Name=dyn_coat_back_side_04_rt
  bone2Name=dyn_coat_back_side_03_rt
  swing1LimitDegrees=15
  swing2LimitDegrees=0
  twistLimitDegrees=45
  twistMotion=Locked
 }
 {
  bone1Name=dyn_coat_front_side_02_lf
  bone2Name=dyn_coat_front_side_01_lf
  swing1LimitDegrees=5
  swing2LimitDegrees=0
  twistLimitDegrees=45
  twistMotion=Locked
 }
 {
  bone1Name=dyn_coat_front_side_03_lf
  bone2Name=dyn_coat_front_side_02_lf
  swing1LimitDegrees=15
  swing2LimitDegrees=0
  twistLimitDegrees=45
  twistMotion=Locked
 }
 {
  bone1Name=dyn_coat_front_side_04_lf
  bone2Name=dyn_coat_front_side_03_lf
  swing1LimitDegrees=25
  swing2LimitDegrees=5
  twistLimitDegrees=45
  twistMotion=Locked
 }
 {
  bone1Name=dyn_coat_front_side_02_rt
  bone2Name=dyn_coat_front_side_01_rt
  swing1LimitDegrees=10
  swing2LimitDegrees=0
  twistLimitDegrees=45
  twistMotion=Locked
 }
 {
  bone1Name=dyn_coat_front_side_03_rt
  bone2Name=dyn_coat_front_side_02_rt
  swing1LimitDegrees=20
  swing2LimitDegrees=5
  twistLimitDegrees=45
  twistMotion=Locked
 }
 {
  bone1Name=dyn_coat_front_side_04_rt
  bone2Name=dyn_coat_front_side_03_rt
  swing1LimitDegrees=25
  swing2LimitDegrees=5
  twistLimitDegrees=45
  twistMotion=Locked
 }
 {
  bone1Name=dyn_coat_side_02_lf
  bone2Name=dyn_coat_side_01_lf
  swing1LimitDegrees=10
  swing2LimitDegrees=0
  twistLimitDegrees=45
  twistMotion=Locked
 }
 {bone1Name=dyn_coat_side_03_lf, bone2Name=dyn_coat_side_02_lf, swing1LimitDegrees=10, swing2LimitDegrees=2, twistLimitDegrees=0}
 {bone1Name=dyn_coat_side_04_lf, bone2Name=dyn_coat_side_03_lf, swing1LimitDegrees=15, swing2LimitDegrees=2, twistLimitDegrees=10}
 {
  bone1Name=dyn_coat_side_02_rt
  bone2Name=dyn_coat_side_01_rt
  swing1LimitDegrees=10
  swing2LimitDegrees=2.5
  twistLimitDegrees=45
  twistMotion=Locked
 }
 {bone1Name=dyn_coat_side_03_rt, bone2Name=dyn_coat_side_02_rt, swing1LimitDegrees=10, swing2LimitDegrees=2, twistLimitDegrees=0}
 {bone1Name=dyn_coat_side_04_rt, bone2Name=dyn_coat_side_03_rt, swing1LimitDegrees=15, swing2LimitDegrees=2, twistLimitDegrees=10}
 {
  bone1Name=dyn_coat_front_01_rt
  bone2Name=offset_coat_front_rt
  swing1LimitDegrees=5
  swing2LimitDegrees=0
  twistLimitDegrees=45
  twistMotion=Locked
 }
 {bone1Name=dyn_coat_front_02_rt, bone2Name=dyn_coat_front_01_rt, swing1LimitDegrees=10, swing2LimitDegrees=0, twistLimitDegrees=5}
 {bone1Name=dyn_coat_front_03_rt, bone2Name=dyn_coat_front_02_rt, swing1LimitDegrees=20, swing2LimitDegrees=0, twistLimitDegrees=5}
 {bone1Name=dyn_coat_front_04_rt, bone2Name=dyn_coat_front_03_rt, swing1LimitDegrees=25, swing2LimitDegrees=0, twistLimitDegrees=0}
 {bone1Name=dyn_coat_front_05_rt, bone2Name=dyn_coat_front_04_rt, swing1LimitDegrees=20, swing2LimitDegrees=2.5, twistLimitDegrees=0}
 {bone1Name=dyn_coat_front_05_lf, bone2Name=dyn_coat_front_04_lf, swing1LimitDegrees=20, swing2LimitDegrees=5, twistLimitDegrees=10}
 {
  bone1Name=dyn_coat_front_03_lf
  bone2Name=dyn_coat_front_02_lf
  swing1LimitDegrees=20
  swing2LimitDegrees=5
  twistLimitDegrees=45
  twistMotion=Locked
 }
 {bone1Name=dyn_coat_front_04_lf, bone2Name=dyn_coat_front_03_lf, swing1LimitDegrees=20, swing2LimitDegrees=5, twistLimitDegrees=10}
 {
  bone1Name=dyn_coat_side_01_lf
  bone2Name=offset_coat_side_lf
  swing1LimitDegrees=10
  swing2LimitDegrees=0
  twistLimitDegrees=45
  twistMotion=Locked
 }
 {
  bone1Name=dyn_coat_side_01_rt
  bone2Name=offset_coat_side_rt
  swing1LimitDegrees=10
  swing2LimitDegrees=0
  twistLimitDegrees=45
  twistMotion=Locked
 }
 {
  bone1Name=dyn_coat_front_side_01_rt
  bone2Name=offset_coat_front_side_rt
  swing1LimitDegrees=10
  swing2LimitDegrees=0
  twistLimitDegrees=45
  twistMotion=Locked
 }
]
