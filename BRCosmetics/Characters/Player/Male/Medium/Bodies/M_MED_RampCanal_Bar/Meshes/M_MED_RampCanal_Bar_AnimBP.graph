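# Character-part AnimBP for the M_MED_RampCanal_Bar player body, adding secondary motion to its coat, hood and belt.
# The copied parent pose runs through a Control Rig and additive coat-bone corrections.
# Rigid-body simulation replaces cloth when bClothEnabled is off, then hood AnimDynamics and belt rigid body follow.
# The event graph sets movement flags and gravity override, and drives the cloth controller with separate emote and normal settings.
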AnimBlueprint /BRCosmetics/Characters/Player/Male/Medium/Bodies/M_MED_RampCanal_Bar/Meshes/M_MED_RampCanal_Bar_AnimBP
target_skeleton=&/M_MED_RampCanal_Bar_Skeleton
graphs [
 AnimGraph {nodes [
   Output Pose(Result=Use cached pose 'DynamicsPose'()) {layerGroup=None}
   n1 = RigidBody!(ComponentPose=Local To Component(LocalPose=Use cached pose 'PreDynamicsPose'()), Alpha, OverrideWorldGravity:Vector="0,0,0", ExternalForce) -> (Pose) {
    bClampLinearTranslationLimitToRefPose=true
    bOverrideWorldGravity=true
    overridePhysicsAsset=&/M_MED_RampCanal_Bar_Physics
    simulationSpace=BaseBoneSpace
    baseBoneRef {boneName=pelvis}
    simSpaceSettings {
     dampingAlpha=0
     maxAngularAcceleration=9
     maxAngularVelocity=4
     maxLinearAcceleration=1000
     maxLinearVelocity=400
     velocityScaleZ=0.75
     worldAlpha=1
     externalLinearDragV=<1.1 1.1 1.1>
    }
   }
   n2 = Component To Local(ComponentPose=@n1)
   n3 = "Transform (Modify) Bone - Bone: glute_l_cloth_helper"!(
    ComponentPose=Local To Component(LocalPose=Use cached pose 'PreDynamicsPose'())
    bAlphaBoolEnabled:bool=True
    Translation="0.3,0,0"
    Rotation="0,0,0"
    Scale="1,1,1"
   ) -> (Pose) {
    alphaInputType=Bool
    translationMode=BMM_Additive
    translationSpace=BCS_BoneSpace
    alphaBoolBlend {blendInTime=0.5, blendOutTime=0.5}
    boneToModify {boneName=glute_l_cloth_helper}
    translation {x=0.3}
   }
   n4 = Component To Local(ComponentPose=@n3)
   n5 = Get bClothEnabled()
   n6 = Anim Dynamics - Bone: dyn_hood_cr!(
    ComponentPose=Local To Component(LocalPose=Blend Poses by bool(BlendPose_0=@n4, BlendPose_1=@n2, bActiveValue=@n5) {blendType=Linear})
    Alpha=1
    ComponentLinearAccScale:Vector="1,1,1"
    ComponentLinearVelScale:Vector="1,1,1"
    ExternalForce="0,0,0"
   ) -> (Pose) {
    angularDampingOverride=1
    angularSpringConstant=1000
    bAngularSpring=true
    bOverrideAngularDamping=true
    bUseGravityOverride=true
    gravityScale=0.35
    simulationSpace=RootRelative
    boundBone {boneName=dyn_hood_cr}
    componentLinearAccScale=<1 1 1>
    componentLinearVelScale=<1 1 1>
    gravityOverride {z=-980}
    physicsBodyDefinitions [{
      boundBone {boneName=dyn_hood_cr}
      boxExtents=<30 30 30>
      constraintSetup {angularLimitsMax=<0 2 5>, angularLimitsMin=<0 -20 -5>, angularTarget=<1 0 0>}
      localJointOffset=<1.5 0 0>
     }]
    simSpaceSettings {maxAngularAcceleration=35, maxAngularVelocity=7, simSpaceAngularAlpha=1}
   }
   n7 = RigidBody(ComponentPose=@n6, OverrideWorldGravity="0,0,0") {
    bClampLinearTranslationLimitToRefPose=true
    bOverrideWorldGravity=true
    overridePhysicsAsset=&/M_MED_RampCanal_Bar_Belt_Physics
    simulationSpace=BaseBoneSpace
    baseBoneRef {boneName=spine_02}
    simSpaceSettings {worldAlpha=0.45}
   }
   n8 = "Transform (Modify) Bone - Bone: base_belt_front_rt"!(ComponentPose=@n7, bAlphaBoolEnabled:bool=True, Translation="0,0,0", Rotation="0,0,0", Scale="0.5,1,0.85") -> (Pose) {
    alphaInputType=Bool
    rotationMode=BMM_Additive
    rotationSpace=BCS_BoneSpace
    scaleMode=BMM_Additive
    scaleSpace=BCS_BoneSpace
    translationMode=BMM_Additive
    translationSpace=BCS_BoneSpace
    alphaBoolBlend {blendInTime=0.4, blendOption=QuadraticInOut, blendOutTime=0.2}
    boneToModify {boneName=base_belt_front_rt}
    scale {x=0.5, z=0.85}
   }
   n9 = "Transform (Modify) Bone - Bone: base_belt_back_rt"!(ComponentPose=@n8, bAlphaBoolEnabled:bool=True, Translation="0,0,0", Rotation="0,0,0", Scale="0.85,1,1") -> (Pose) {
    alphaInputType=Bool
    rotationMode=BMM_Additive
    rotationSpace=BCS_BoneSpace
    scaleMode=BMM_Additive
    scaleSpace=BCS_BoneSpace
    translationMode=BMM_Additive
    translationSpace=BCS_BoneSpace
    alphaBoolBlend {blendInTime=0.4, blendOption=QuadraticInOut, blendOutTime=0.2}
    boneToModify {boneName=base_belt_back_rt}
    scale {x=0.85}
   }
   Save cached pose 'DynamicsPose'(Pose=Component To Local(ComponentPose=@n9)) {cacheName=DynamicsPose}
   n10 = Control Rig!(
    Source=Copy Pose From Mesh() {bUseAttachedParent=true}
    Alpha
    IsCrouch:bool=Get CrouchIDLE() -> (CrouchIDLE:bool)
    IsCrouchWalking:bool=Get IsCrouchWalking() -> (IsCrouchWalking:bool)
    IsJumping:bool=Get IsJumpingCustom() -> (IsJumpingCustom:bool)
    IsCloth:bool=Get bClothEnabled()
    IsOnAir:bool=Get bIsSkyDiving()
   ) -> (Pose) {controlRigAssetReference {blueprintRigClass=&/M_MED_RampCanal_Bar_CtrlRig.M_MED_RampCanal_Bar_CtrlRig_C}}
   Save cached pose 'ControlRigPose'(Pose=@n10) {cacheName=ControlRigPose}
   n11 = "Transform (Modify) Bone - Bone: base_coat_back_side_lf"!(
    ComponentPose=Local To Component(LocalPose=Use cached pose 'ControlRigPose'())
    bAlphaBoolEnabled:bool=True
    Translation="0,0,0"
    Scale="1,1,1"
   ) -> (Pose) {
    alphaInputType=Bool
    rotationMode=BMM_Additive
    rotationSpace=BCS_BoneSpace
    alphaBoolBlend {blendInTime=0.4, blendOutTime=0.4}
    boneToModify {boneName=base_coat_back_side_lf}
    rotation=<rot -5.0004 -4.8673 -0.0984>
   }
   n12 = "Transform (Modify) Bone - Bone: base_coat_back_lf"!(ComponentPose=@n11, bAlphaBoolEnabled:bool=True, Translation="0,0,0", Scale="1,1,1") -> (Pose) {
    alphaInputType=Bool
    rotationMode=BMM_Additive
    rotationSpace=BCS_BoneSpace
    alphaBoolBlend {blendInTime=0.4, blendOutTime=0.4}
    boneToModify {boneName=base_coat_back_lf}
    rotation {yaw=-20}
   }
   n13 = "Transform (Modify) Bone - Bone: base_coat_back_side_rt"!(ComponentPose=@n12, bAlphaBoolEnabled:bool=True, Translation="0,0,0", Scale="1,1,1") -> (Pose) {
    alphaInputType=Bool
    rotationMode=BMM_Additive
    rotationSpace=BCS_BoneSpace
    alphaBoolBlend {blendInTime=0.4, blendOutTime=0.4}
    boneToModify {boneName=base_coat_back_side_rt}
    rotation=<rot -0.1511 -19.9773 0.0166>
   }
   n14 = "Transform (Modify) Bone - Bone: base_coat_back_rt"!(ComponentPose=@n13, bAlphaBoolEnabled:bool=True, Translation="0,0,0", Scale="1,1,1") -> (Pose) {
    alphaInputType=Bool
    rotationMode=BMM_Additive
    rotationSpace=BCS_BoneSpace
    alphaBoolBlend {blendInTime=0.4, blendOutTime=0.4}
    boneToModify {boneName=base_coat_back_rt}
    rotation {yaw=-20}
   }
   n15 = "Transform (Modify) Bone - Bone: offset_coat_front_side_lf"!(ComponentPose=@n14, bAlphaBoolEnabled:bool=True, Translation="0,0,0", Scale="1,1,1") -> (Pose) {
    alphaInputType=Bool
    rotationMode=BMM_Additive
    rotationSpace=BCS_BoneSpace
    alphaBoolBlend {blendInTime=0.4, blendOutTime=0.4}
    boneToModify {boneName=offset_coat_front_side_lf}
    rotation=<rot -14.5192 5.2438 -3.3418>
   }
   n16 = "Transform (Modify) Bone - Bone: offset_coat_front_side_rt"!(ComponentPose=@n15, bAlphaBoolEnabled:bool=True, Translation="0,0,0", Scale="1,1,1") -> (Pose) {
    alphaInputType=Bool
    rotationMode=BMM_Additive
    rotationSpace=BCS_BoneSpace
    alphaBoolBlend {blendInTime=0.4, blendOutTime=0.4}
    boneToModify {boneName=offset_coat_front_side_rt}
    rotation=<rot 3.7904 -13.0908 -9.7485>
   }
   n17 = "Transform (Modify) Bone - Bone: base_coat_front_rt"!(ComponentPose=@n16, bAlphaBoolEnabled:bool=True, Translation="0,0,0", Scale="1,1,1") -> (Pose) {
    alphaInputType=Bool
    rotationMode=BMM_Additive
    rotationSpace=BCS_BoneSpace
    alphaBoolBlend {blendInTime=0.4, blendOutTime=0.4}
    boneToModify {boneName=base_coat_front_rt}
   }
   n18 = "Transform (Modify) Bone - Bone: offset_coat_front_lf"!(ComponentPose=@n17, bAlphaBoolEnabled:bool=True, Translation="0,0,0", Scale="1,1,1") -> (Pose) {
    alphaInputType=Bool
    rotationMode=BMM_Additive
    rotationSpace=BCS_BoneSpace
    alphaBoolBlend {blendInTime=0.4, blendOutTime=0.4}
    boneToModify {boneName=offset_coat_front_lf}
    rotation=<rot 0.3282 -14.9346 1.3058>
   }
   n19 = "Transform (Modify) Bone - Bone: offset_coat_side_lf"!(ComponentPose=@n18, bAlphaBoolEnabled:bool=True, Translation="0,0,0", Scale="1,1,1") -> (Pose) {
    alphaInputType=Bool
    rotationMode=BMM_Additive
    rotationSpace=BCS_BoneSpace
    alphaBoolBlend {blendInTime=0.4, blendOutTime=0.4}
    boneToModify {boneName=offset_coat_side_lf}
    rotation=<rot -9.9999 -0.0293 0.0148>
   }
   n20 = "Transform (Modify) Bone - Bone: offset_coat_front_side_rt"(ComponentPose=@n19, bAlphaBoolEnabled=True, Translation="0,0,0", Scale="1,1,1") {
    alphaInputType=Bool
    rotationMode=BMM_Additive
    rotationSpace=BCS_BoneSpace
    alphaBoolBlend {blendInTime=0.4, blendOutTime=0.4}
    boneToModify {boneName=offset_coat_front_side_rt}
    rotation=<rot 4.9995 0.0033 -0.0701>
   }
   n21 = "Transform (Modify) Bone - Bone: offset_coat_side_rt"!(ComponentPose=@n20, bAlphaBoolEnabled:bool=True, Translation="0,0,0", Scale="1,1,1") -> (Pose) {
    alphaInputType=Bool
    rotationMode=BMM_Additive
    rotationSpace=BCS_BoneSpace
    alphaBoolBlend {blendInTime=0.4, blendOutTime=0.4}
    boneToModify {boneName=offset_coat_side_rt}
    rotation=<rot 14.997 -0.1566 -0.6899>
   }
   n22 = "Transform (Modify) Bone - Bone: offset_coat_side_rt"(ComponentPose=@n21, bAlphaBoolEnabled=True, Translation="0,0,0", Scale="1,1,1") {
    alphaInputType=Bool
    rotationMode=BMM_Additive
    rotationSpace=BCS_BoneSpace
    alphaBoolBlend {blendInTime=0.4, blendOutTime=0.4}
    boneToModify {boneName=offset_coat_side_rt}
    rotation {yaw=-20}
   }
   n23 = "Transform (Modify) Bone - Bone: offset_coat_front_rt"!(ComponentPose=@n22, bAlphaBoolEnabled:bool=True, Translation="0,0,0", Scale="1,1,1") -> (Pose) {
    alphaInputType=Bool
    rotationMode=BMM_Additive
    rotationSpace=BCS_BoneSpace
    alphaBoolBlend {blendInTime=0.4, blendOutTime=0.4}
    boneToModify {boneName=offset_coat_front_rt}
    rotation {pitch=10, yaw=-10}
   }
   n24 = "Transform (Modify) Bone - Bone: base_coat_front_side_rt"!(ComponentPose=@n23, bAlphaBoolEnabled:bool=True, Translation="0,0,0", Scale="1,1,1") -> (Pose) {
    alphaInputType=Bool
    rotationMode=BMM_Additive
    rotationSpace=BCS_BoneSpace
    alphaBoolBlend {blendInTime=0.4, blendOutTime=0.4}
    boneToModify {boneName=base_coat_front_side_rt}
    rotation=<rot 14.5081 -60 -3.9243>
   }
   n25 = "Transform (Modify) Bone - Bone: offset_coat_side_rt"(ComponentPose=@n24, bAlphaBoolEnabled=True, Translation="0,0,0", Scale="1,1,1") {
    alphaInputType=Bool
    rotationMode=BMM_Additive
    rotationSpace=BCS_BoneSpace
    alphaBoolBlend {blendInTime=0.4, blendOutTime=0.4}
    boneToModify {boneName=offset_coat_side_rt}
    rotation {yaw=-40}
   }
   n26 = "Transform (Modify) Bone - Bone: base_coat_back_side_lf"(ComponentPose=@n25, bAlphaBoolEnabled=True, Translation="0,0,0", Scale="1,1,1") {
    alphaInputType=Bool
    rotationMode=BMM_Additive
    rotationSpace=BCS_BoneSpace
    alphaBoolBlend {blendInTime=0.4, blendOutTime=0.4}
    boneToModify {boneName=base_coat_back_side_lf}
    rotation {yaw=-50}
   }
   n27 = "Transform (Modify) Bone - Bone: base_coat_back_lf"(ComponentPose=@n26, bAlphaBoolEnabled=True, Translation="0,0,0", Scale="1,1,1") {
    alphaInputType=Bool
    rotationMode=BMM_Additive
    rotationSpace=BCS_BoneSpace
    alphaBoolBlend {blendInTime=0.4, blendOutTime=0.4}
    boneToModify {boneName=base_coat_back_lf}
    rotation {yaw=-50}
   }
   n28 = "Transform (Modify) Bone - Bone: base_coat_back_side_rt"(ComponentPose=@n27, bAlphaBoolEnabled=True, Translation="0,0,0", Scale="1,1,1") {
    alphaInputType=Bool
    rotationMode=BMM_Additive
    rotationSpace=BCS_BoneSpace
    alphaBoolBlend {blendInTime=0.4, blendOutTime=0.4}
    boneToModify {boneName=base_coat_back_side_rt}
    rotation {yaw=-50}
   }
   n29 = "Transform (Modify) Bone - Bone: base_coat_back_rt"(ComponentPose=@n28, bAlphaBoolEnabled=True, Translation="0,0,0", Scale="1,1,1") {
    alphaInputType=Bool
    rotationMode=BMM_Additive
    rotationSpace=BCS_BoneSpace
    alphaBoolBlend {blendInTime=0.4, blendOutTime=0.4}
    boneToModify {boneName=base_coat_back_rt}
    rotation {yaw=-50}
   }
   n30 = "Transform (Modify) Bone - Bone: offset_coat_side_rt"(ComponentPose=@n29, bAlphaBoolEnabled=True, Translation="0,0,0", Scale="1,1,1") {
    alphaInputType=Bool
    rotationMode=BMM_Additive
    rotationSpace=BCS_BoneSpace
    alphaBoolBlend {blendInTime=0.4, blendOutTime=0.4}
    boneToModify {boneName=offset_coat_side_rt}
   }
   n31 = "Transform (Modify) Bone - Bone: base_coat_back_side_lf"!(ComponentPose=@n30, bAlphaBoolEnabled=True, Scale="1,1,1") -> (Pose) {
    alphaInputType=Bool
    rotationMode=BMM_Additive
    rotationSpace=BCS_BoneSpace
    translationMode=BMM_Additive
    translationSpace=BCS_BoneSpace
    alphaBoolBlend {blendInTime=0.4, blendOutTime=0.2}
    boneToModify {boneName=base_coat_back_side_lf}
    rotation=<rot -2.0791 10.6148 -10.9162>
    translation=<0.4002 -2.2769 -0.8224>
   }
   n32 = "Transform (Modify) Bone - Bone: offset_coat_side_lf"!(ComponentPose=@n31, bAlphaBoolEnabled=True, Scale="1,1,1") -> (Pose) {
    alphaInputType=Bool
    rotationMode=BMM_Additive
    rotationSpace=BCS_BoneSpace
    translationMode=BMM_Additive
    translationSpace=BCS_BoneSpace
    alphaBoolBlend {blendInTime=0.4, blendOutTime=0.2}
    boneToModify {boneName=offset_coat_side_lf}
    rotation=<rot 11.1564 8.2567 -35.574>
    translation=<0.2724 -1.1523 -4.939>
   }
   n33 = "Transform (Modify) Bone - Bone: base_coat_back_side_rt"!(ComponentPose=@n32, bAlphaBoolEnabled=True, Scale="1,1,1") -> (Pose) {
    alphaInputType=Bool
    rotationMode=BMM_Additive
    rotationSpace=BCS_BoneSpace
    translationMode=BMM_Additive
    translationSpace=BCS_BoneSpace
    alphaBoolBlend {blendInTime=0.4, blendOutTime=0.2}
    boneToModify {boneName=base_coat_back_side_rt}
    rotation=<rot -0.0024 -5 0.0082>
   }
   n34 = "Transform (Modify) Bone - Bone: base_coat_back_rt"!(ComponentPose=@n33, bAlphaBoolEnabled=True, Scale="1,1,1") -> (Pose) {
    alphaInputType=Bool
    rotationMode=BMM_Additive
    rotationSpace=BCS_BoneSpace
    translationMode=BMM_Additive
    translationSpace=BCS_BoneSpace
    alphaBoolBlend {blendInTime=0.4, blendOutTime=0.2}
    boneToModify {boneName=base_coat_back_rt}
   }
   n35 = "Transform (Modify) Bone - Bone: base_coat_side_rt"!(ComponentPose=@n34, bAlphaBoolEnabled:bool=True, Scale="1,1,1") -> (Pose) {
    alphaInputType=Bool
    rotationMode=BMM_Additive
    rotationSpace=BCS_BoneSpace
    translationMode=BMM_Additive
    translationSpace=BCS_BoneSpace
    alphaBoolBlend {blendInTime=0.4, blendOutTime=0.2}
    boneToModify {boneName=base_coat_side_rt}
    rotation=<rot 0.8733 -18.3926 -22.1279>
   }
   n36 = "Transform (Modify) Bone - Bone: base_coat_front_side_rt"!(ComponentPose=@n35, bAlphaBoolEnabled=True, Scale="1,1,1") -> (Pose) {
    alphaInputType=Bool
    rotationMode=BMM_Additive
    rotationSpace=BCS_BoneSpace
    translationMode=BMM_Additive
    translationSpace=BCS_BoneSpace
    alphaBoolBlend {blendInTime=0.4, blendOutTime=0.2}
    boneToModify {boneName=base_coat_front_side_rt}
    rotation=<rot 0.0636 -15 0.0119>
   }
   n37 = "Transform (Modify) Bone - Bone: base_coat_front_rt"!(ComponentPose=@n36, bAlphaBoolEnabled=True, Scale="1,1,1") -> (Pose) {
    alphaInputType=Bool
    rotationMode=BMM_Additive
    rotationSpace=BCS_BoneSpace
    translationMode=BMM_Additive
    translationSpace=BCS_BoneSpace
    alphaBoolBlend {blendInTime=0.4, blendOutTime=0.2}
    boneToModify {boneName=base_coat_front_rt}
    rotation=<rot -48.9025 -30.1817 11.5724>
   }
   n38 = "Transform (Modify) Bone - Bone: offset_coat_front_lf"!(ComponentPose=@n37, bAlphaBoolEnabled=True, Scale="1,1,1") -> (Pose) {
    alphaInputType=Bool
    rotationMode=BMM_Additive
    rotationSpace=BCS_BoneSpace
    translationMode=BMM_Additive
    translationSpace=BCS_BoneSpace
    alphaBoolBlend {blendInTime=0.4, blendOutTime=0.2}
    boneToModify {boneName=offset_coat_front_lf}
    rotation=<rot -0.0819 -9.9999 -0.0683>
   }
   n39 = "Transform (Modify) Bone - Bone: offset_coat_front_side_lf"!(ComponentPose=@n38, bAlphaBoolEnabled=True, Scale="1,1,1") -> (Pose) {
    alphaInputType=Bool
    rotationMode=BMM_Additive
    rotationSpace=BCS_BoneSpace
    translationMode=BMM_Additive
    translationSpace=BCS_BoneSpace
    alphaBoolBlend {blendInTime=0.4, blendOutTime=0.2}
    boneToModify {boneName=offset_coat_front_side_lf}
    rotation=<rot 3.6678 1.9102 -37.6585>
   }
   n40 = "Transform (Modify) Bone - Bone: base_coat_back_side_lf"!(ComponentPose=@n39, bAlphaBoolEnabled=True, Translation:Vector="0,0,0", Scale="1,1,1") -> (Pose) {
    alphaInputType=Bool
    rotationMode=BMM_Additive
    rotationSpace=BCS_BoneSpace
    alphaBoolBlend {blendInTime=0.4, blendOutTime=0.2}
    boneToModify {boneName=base_coat_back_side_lf}
    rotation=<rot 0 -69.9997 -0.0002>
   }
   n41 = "Transform (Modify) Bone - Bone: base_coat_back_lf"(ComponentPose=@n40, bAlphaBoolEnabled=True, Translation="0,0,0", Scale="1,1,1") {
    alphaInputType=Bool
    rotationMode=BMM_Additive
    rotationSpace=BCS_BoneSpace
    alphaBoolBlend {blendInTime=0.4, blendOutTime=0.2}
    boneToModify {boneName=base_coat_back_lf}
    rotation=<rot -5 -74.9998 0>
   }
   n42 = "Transform (Modify) Bone - Bone: base_coat_back_side_rt"!(ComponentPose=@n41, bAlphaBoolEnabled=True, Translation:Vector="0,0,0", Scale="1,1,1") -> (Pose) {
    alphaInputType=Bool
    rotationMode=BMM_Additive
    rotationSpace=BCS_BoneSpace
    alphaBoolBlend {blendInTime=0.4, blendOutTime=0.2}
    boneToModify {boneName=base_coat_back_side_rt}
    rotation=<rot -13.1642 -51.0187 6.949>
   }
   n43 = "Transform (Modify) Bone - Bone: base_coat_back_rt"!(ComponentPose=@n42, bAlphaBoolEnabled=True, Translation:Vector="0,0,0", Scale="1,1,1") -> (Pose) {
    alphaInputType=Bool
    rotationMode=BMM_Additive
    rotationSpace=BCS_BoneSpace
    alphaBoolBlend {blendInTime=0.4, blendOutTime=0.2}
    boneToModify {boneName=base_coat_back_rt}
    rotation=<rot 2.5881 -64.9618 4.3823>
   }
   n44 = "Transform (Modify) Bone - Bone: base_coat_side_rt"!(ComponentPose=@n43, bAlphaBoolEnabled=True, Translation:Vector="0,0,0", Scale="1,1,1") -> (Pose) {
    alphaInputType=Bool
    rotationMode=BMM_Additive
    rotationSpace=BCS_BoneSpace
    alphaBoolBlend {blendInTime=0.4, blendOutTime=0.2}
    boneToModify {boneName=base_coat_side_rt}
    rotation=<rot -15.1889 -33.1677 41.7632>
   }
   n45 = "Transform (Modify) Bone - Bone: base_coat_side_lf"!(ComponentPose=@n44, bAlphaBoolEnabled:bool=True, Translation="0,0,0", Scale="1,1,1") -> (Pose) {
    alphaInputType=Bool
    rotationMode=BMM_Additive
    rotationSpace=BCS_BoneSpace
    alphaBoolBlend {blendInTime=0.4, blendOutTime=0.2}
    boneToModify {boneName=base_coat_side_lf}
    rotation=<rot -34.4911 -54.5997 27.7252>
   }
   n46 = "Transform (Modify) Bone - Bone: coat_corr_front_01_lf"!(ComponentPose=@n45, bAlphaBoolEnabled:bool=True, Translation="10.294,-0.078,3.655", Scale="1,1,1") -> (Pose) {
    alphaInputType=Bool
    rotationMode=BMM_Additive
    rotationSpace=BCS_BoneSpace
    translationMode=BMM_Additive
    translationSpace=BCS_BoneSpace
    alphaBoolBlend {blendInTime=0.4, blendOutTime=0.2}
    boneToModify {boneName=coat_corr_front_01_lf}
    rotation=<rot 27.4415 61.1765 0.3169>
    translation=<10.294 -0.078 3.655>
   }
   n47 = "Transform (Modify) Bone - Bone: coat_corr_front_01_rt"!(ComponentPose=@n46, bAlphaBoolEnabled:bool=True, Translation="4.853,-0.022,2.418", Scale="1,1,1") -> (Pose) {
    alphaInputType=Bool
    rotationMode=BMM_Additive
    rotationSpace=BCS_BoneSpace
    translationMode=BMM_Additive
    translationSpace=BCS_BoneSpace
    alphaBoolBlend {blendInTime=0.4, blendOutTime=0.2}
    boneToModify {boneName=coat_corr_front_01_rt}
    rotation=<rot 74.5516 -0.976 -40.6981>
    translation=<4.853 -0.022 2.418>
   }
   n48 = "Transform (Modify) Bone - Bone: offset_coat_front_rt"(ComponentPose=@n47, bAlphaBoolEnabled=True, Translation="-0.675,11.834,0.000", Scale="1,1,1") {
    alphaInputType=Bool
    rotationMode=BMM_Additive
    rotationSpace=BCS_BoneSpace
    translationMode=BMM_Additive
    translationSpace=BCS_BoneSpace
    alphaBoolBlend {blendInTime=0.4, blendOutTime=0.2}
    boneToModify {boneName=offset_coat_front_rt}
    rotation=<rot 75 0 0.0001>
    translation=<-0.675 11.834>
   }
   n49 = "Transform (Modify) Bone - Bone: offset_coat_front_lf"!(ComponentPose=@n48, bAlphaBoolEnabled=True, Translation:Vector="15.137,3.036,-1.153", Scale="1,1,1") -> (Pose) {
    alphaInputType=Bool
    rotationMode=BMM_Additive
    rotationSpace=BCS_BoneSpace
    translationMode=BMM_Additive
    translationSpace=BCS_BoneSpace
    alphaBoolBlend {blendInTime=0.4, blendOutTime=0.2}
    boneToModify {boneName=offset_coat_front_lf}
    rotation=<rot -50.2536 38.1178 -79.8646>
    translation=<15.137 3.036 -1.153>
   }
   n50 = "Transform (Modify) Bone - Bone: offset_coat_front_side_rt"(ComponentPose=@n49, bAlphaBoolEnabled=True, Translation="3.237,8.635,1.736", Scale="1,1,1") {
    alphaInputType=Bool
    rotationMode=BMM_Additive
    rotationSpace=BCS_BoneSpace
    translationMode=BMM_Additive
    translationSpace=BCS_BoneSpace
    alphaBoolBlend {blendInTime=0.4, blendOutTime=0.2}
    boneToModify {boneName=offset_coat_front_side_rt}
    rotation=<rot 49.7765 -27.7701 -40.6885>
    translation=<3.237 8.635 1.736>
   }
   n51 = "Transform (Modify) Bone - Bone: offset_coat_front_side_lf"!(ComponentPose=@n50, bAlphaBoolEnabled=True, Translation:Vector="5.072,0.000,-0.000", Scale="1,1,1") -> (Pose) {
    alphaInputType=Bool
    rotationMode=BMM_Additive
    rotationSpace=BCS_BoneSpace
    translationMode=BMM_Additive
    translationSpace=BCS_BoneSpace
    alphaBoolBlend {blendInTime=0.4, blendOutTime=0.2}
    boneToModify {boneName=offset_coat_front_side_lf}
    rotation=<rot -55.4831 -9.5415 -24.0663>
    translation {x=5.072}
   }
   n52 = "Transform (Modify) Bone - Bone: base_coat_back_side_lf"(ComponentPose=@n51, bAlphaBoolEnabled=True, Translation="0,0,0", Scale="1,1,1") {
    alphaInputType=Bool
    rotationMode=BMM_Additive
    rotationSpace=BCS_BoneSpace
    alphaBoolBlend {blendInTime=0.4, blendOutTime=0.2}
    boneToModify {boneName=base_coat_back_side_lf}
    rotation=<rot -22.5209 -62.2262 11.148>
   }
   n53 = "Transform (Modify) Bone - Bone: base_coat_back_lf"(ComponentPose=@n52, bAlphaBoolEnabled=True, Translation="0,0,0", Scale="1,1,1") {
    alphaInputType=Bool
    rotationMode=BMM_Additive
    rotationSpace=BCS_BoneSpace
    alphaBoolBlend {blendInTime=0.4, blendOutTime=0.2}
    boneToModify {boneName=base_coat_back_lf}
    rotation=<rot 7.4559 -115.9611 38.369>
   }
   n54 = "Transform (Modify) Bone - Bone: base_coat_back_side_rt"(ComponentPose=@n53, bAlphaBoolEnabled=True, Translation="0,0,0", Scale="1,1,1") {
    alphaInputType=Bool
    rotationMode=BMM_Additive
    rotationSpace=BCS_BoneSpace
    alphaBoolBlend {blendInTime=0.4, blendOutTime=0.2}
    boneToModify {boneName=base_coat_back_side_rt}
    rotation=<rot -0.0001 -94.9995 -0.0002>
   }
   n55 = "Transform (Modify) Bone - Bone: base_coat_back_rt"(ComponentPose=@n54, bAlphaBoolEnabled=True, Translation="0,0,0", Scale="1,1,1") {
    alphaInputType=Bool
    rotationMode=BMM_Additive
    rotationSpace=BCS_BoneSpace
    translationMode=BMM_Additive
    translationSpace=BCS_BoneSpace
    alphaBoolBlend {blendInTime=0.4, blendOutTime=0.2}
    boneToModify {boneName=base_coat_back_rt}
    rotation=<rot 2.8944 -93.8477 31.0199>
   }
   n56 = "Transform (Modify) Bone - Bone: base_coat_side_rt"(ComponentPose=@n55, bAlphaBoolEnabled=True, Translation="0,0,0", Scale="1,1,1") {
    alphaInputType=Bool
    rotationMode=BMM_Additive
    rotationSpace=BCS_BoneSpace
    translationMode=BMM_Additive
    translationSpace=BCS_BoneSpace
    alphaBoolBlend {blendInTime=0.4, blendOutTime=0.2}
    boneToModify {boneName=base_coat_side_rt}
    rotation=<rot 0 -124.9997 -0.0001>
   }
   n57 = "Transform (Modify) Bone - Bone: base_coat_front_side_rt"!(ComponentPose=@n56, bAlphaBoolEnabled=True, Translation:Vector="0.693,-1.188,-0.478", Scale="1,1,1") -> (Pose) {
    alphaInputType=Bool
    rotationMode=BMM_Additive
    rotationSpace=BCS_BoneSpace
    translationMode=BMM_Additive
    translationSpace=BCS_BoneSpace
    alphaBoolBlend {blendInTime=0.4, blendOutTime=0.2}
    boneToModify {boneName=base_coat_front_side_rt}
    rotation=<rot 41.0235 -41.241 0.2961>
    translation=<0.693 -1.188 -0.478>
   }
   n58 = "Transform (Modify) Bone - Bone: base_coat_front_rt"!(ComponentPose=@n57, bAlphaBoolEnabled=True, Translation:Vector="-1.730,1.023,8.003", Scale="1,1,1") -> (Pose) {
    alphaInputType=Bool
    rotationMode=BMM_Additive
    rotationSpace=BCS_BoneSpace
    translationMode=BMM_Additive
    translationSpace=BCS_BoneSpace
    alphaBoolBlend {blendInTime=0.4, blendOutTime=0.2}
    boneToModify {boneName=base_coat_front_rt}
    rotation=<rot -5.0001 0 44.9999>
    translation=<-1.73 1.023 8.003>
   }
   n59 = "Transform (Modify) Bone - Bone: offset_coat_front_lf"(ComponentPose=@n58, bAlphaBoolEnabled=True, Translation="13.165,1.747,-0.777", Scale="1,1,1") {
    alphaInputType=Bool
    rotationMode=BMM_Additive
    rotationSpace=BCS_BoneSpace
    translationMode=BMM_Additive
    translationSpace=BCS_BoneSpace
    alphaBoolBlend {blendInTime=0.4, blendOutTime=0.2}
    boneToModify {boneName=offset_coat_front_lf}
    rotation=<rot 14.826 -24.6153 2.5675>
    translation=<13.165 1.747 -0.777>
   }
   n60 = "Transform (Modify) Bone - Bone: coat_corr_front_01_lf"(ComponentPose=@n59, bAlphaBoolEnabled=True, Translation="6.368,1.599,1.180", Scale="1,1,1") {
    alphaInputType=Bool
    rotationMode=BMM_Additive
    rotationSpace=BCS_BoneSpace
    translationMode=BMM_Additive
    translationSpace=BCS_BoneSpace
    alphaBoolBlend {blendInTime=0.4, blendOutTime=0.2}
    boneToModify {boneName=coat_corr_front_01_lf}
    rotation=<rot 3.7627 40.0073 14.5265>
    translation=<6.368 1.599 1.18>
   }
   n61 = "Transform (Modify) Bone - Bone: attach_backpack"!(ComponentPose=@n60, bAlphaBoolEnabled:bool=True, Translation="0,0.7,0", Rotation="0,0,0", Scale="1,1,1") -> (Pose) {
    alphaInputType=Bool
    rotationMode=BMM_Additive
    rotationSpace=BCS_BoneSpace
    translationMode=BMM_Additive
    translationSpace=BCS_BoneSpace
    alphaBoolBlend {blendInTime=0.4, blendOption=QuadraticInOut, blendOutTime=0.2}
    boneToModify {boneName=attach_backpack}
    translation {y=0.7}
   }
   Save cached pose 'PreDynamicsPose'(Pose=Component To Local(ComponentPose=@n61)) {cacheName=PreDynamicsPose}
  ]}
 EventGraph {nodes [
   n1 = Event BlueprintUpdateAnimation()
   > Set DeltaTimeLocal(DeltaTimeLocal:double=@n1.DeltaTimeX) -> (Output_Get:double)
   > n2 = Sequence!(execute) -> (then_0, then_1, then_2:Exec)
   n3 = Get bIsMoving2D()
   Set CrouchIDLE(execute=@n2.then_1, CrouchIDLE:bool=AND Boolean(A=Get bIsCrouching(), B=NOT Boolean(A=@n3))) -> (Output_Get:bool)
   n4 = Get bIsSkyDiving()
   > Set IsJumpingCustom(IsJumpingCustom:bool=AND Boolean(A=Get bIsJumping(), B=NOT Boolean(A=@n4))) -> (Output_Get:bool)
   n5 = Get bIsCrouching()
   n6 = Get bIsDBNO()
   n7 = Get bIsSwimSprinting() -> (bIsSwimSprinting:bool)
   > Set IsGroundMovement(IsGroundMovement:bool=AND Boolean!(A=Get bIsSprinting(), B=NOT Boolean(A=@n7), C:bool=NOT Boolean(A=@n5), D:bool=NOT Boolean(A=@n6)) -> (ReturnValue)) -> (Output_Get:bool)
   > Set IsCrouchWalking(IsCrouchWalking:bool=AND Boolean!(A=Get bIsCrouching(), B=Get bIsMoving2D()) -> (ReturnValue)) -> (Output_Get:bool)
   n8 = GetOwnerAnimationInterface!(self:Custom Character Part Anim Instance Object Reference) -> (ReturnValue:Fort Player Animation Interface Object Reference)
   Is Valid(exec=@n2.then_2, InputObject=@n8)
   > n9 = Branch(Condition=Break Fort Anim Input Zipline!(FortAnimInput_Zipline:"Fort Anim Input Zipline Structure (by ref)"=GetZiplineInput!(self:Fort Player Animation Interface Object Reference=@n8) -> (ReturnValue:"Fort Anim Input Zipline Structure (by ref)")) -> (
     bIsZiplining:bool=~
     bIsOnAscender:bool=~
     bShouldPlayPivotTransition:bool=~
     LeanYaw:float=~
     PivotBlendDelayRemaining:float=~
     PivotBlendDelay:float=~
     PivotCardinalDirection:EFortCardinalDirection Enum
     WorldVelocityLastTick:Vector=~
    ) {is=title K2Node_BreakStruct})
   > Set animDyn on zipline(animDyn on zipline:Vector="0, 0, 0") -> (Output_Get:Vector)
   Set animDyn on zipline(execute=@n9.else, animDyn on zipline="1,1,1")
   n10 = "Gravity Override State Function (Parameter Asset) (Nativized)"(
    execute=@n2.then_0
    CharacterPartAnimInstance=Self-Reference()
    GameTimeInSeconds=GetGameTimeInSeconds()
    Parameters=&/M_MED_RampCanal_Bar_Parameters
   )
   > Set GravityOverride(GravityOverride=@n10)
   > Branch(Condition=Get bClothEnabled())
   > n11 = Sequence!(execute) -> (then_0, then_1)
   "Cloth Controller Function (Parameter Asset) (Nativized)"!(
    execute=@n11.then_1
    CharacterPartAnimInstance:Custom Character Part Anim Instance Object Reference=Self-Reference()
    ClothingAssetName:Name=M_MED_RampCanal_Bar_Clothing
    ClothCoreSettings:"Cloth Core Settings Structure (by ref)"=Get Cloth Core Settings() -> (Cloth Core Settings:Cloth Core Settings Structure)
    WindGustLocalSettings:"Wind Gust Settings Structure (by ref)"=Get Wind Gust Local Settings() -> (Wind Gust Local Settings:Wind Gust Settings Structure)
    ClothUpAxis:Vector="1,0,0"
    LinVelScaleFwdBwdClamp:float=0
    LinVelScaleUpDwnClamp:float=0.7
    DeltaTime:float=Get DeltaTimeLocal() -> (DeltaTimeLocal:double)
    GameTimeInSeconds:float=GetGameTimeInSeconds()
    StateLogicLayerName:Name=none
    Parameters:Fort Character Dynamics Parameters Object Reference=&M_MED_RampCanal_Bar_Parameters
   ) -> (
    then
    OutCurrentState:Name=None
    OutClothParameters:Cloth Parameters Structure
    OutClothCoreSettings:Cloth Core Settings Structure
    OutWindGustSettings:Wind Gust Parameters Structure
    OutWindGustLocalSettings:Wind Gust Settings Structure
    OutWindVelocity:Vector
    OutLinearVelVector:Vector
   )
   n12 = Branch(execute=@n11.then_0, Condition=Get bIsPlayingEmote())
   > Set Cloth Core Settings(Cloth Core Settings:Cloth Core Settings Structure=Get Cloth Core Settings_Emotes() -> (Cloth Core Settings_Emotes:Cloth Core Settings Structure)) -> (Output_Get:Cloth Core Settings Structure)
   Set Cloth Core Settings(execute=@n12.else, Cloth Core Settings=Get Cloth Core Settings_Normal() -> (Cloth Core Settings_Normal:Cloth Core Settings Structure))
  ]}
]
variables [
 animDyn on zipline {type=Vector, default=<1 1 1>}
 Cloth Core Settings {
  type=ClothCoreSettings
  default {
   collisionThickness=0
   dampingCoefficient=0.05
   frictionCoefficient=0.3
   localDampingCoefficient=0
   numberIterations=2
   numberSubsteps=1
   selfCollisionThickness=0
   animDriveDamping=<0 1>
   animDriveStiffness=<0 1>
   areaStiffness=<1 1>
   bendingStiffness=<1 1>
   edgeStiffness=<1 1>
   tetherStiffness=<1 1>
  }
 }
 Cloth Core Settings_Emotes {
  type=ClothCoreSettings
  default {
   collisionThickness=0
   dampingCoefficient=0.05
   frictionCoefficient=0.3
   localDampingCoefficient=0
   numberIterations=2
   numberSubsteps=1
   selfCollisionThickness=0
   animDriveDamping=<0 1>
   animDriveStiffness=<0.25 1>
   areaStiffness=<1 1>
   bendingStiffness=<1 1>
   edgeStiffness=<1 1>
   tetherStiffness=<1 1>
  }
 }
 Cloth Core Settings_Normal {
  type=ClothCoreSettings
  default {
   collisionThickness=0
   dampingCoefficient=0.05
   frictionCoefficient=0.3
   localDampingCoefficient=0
   numberIterations=2
   numberSubsteps=1
   selfCollisionThickness=0
   animDriveDamping=<0 1>
   animDriveStiffness=<0 1>
   areaStiffness=<1 1>
   bendingStiffness=<1 1>
   edgeStiffness=<1 1>
   tetherStiffness=<1 1>
  }
 }
 CrouchIDLE {type=boolean}
 DeltaTimeLocal {type=number}
 GravityOverride {type=Vector}
 IDLE {type=boolean}
 IsCrouchWalking {type=boolean}
 IsGroundMovement {type=boolean}
 IsJumpingCustom {type=boolean}
 Wind Gust Local Settings {type=WindGustSettings, default {elapsedTime=0, finalGustStrengthLocal=0, gustDuration=0, gustStrength=0, timeBetweenGusts=0}}
]
