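# Physics asset of cloth collision shapes for the M_MED_RampCanal_Bar male medium player body.
# Capsules cover the pelvis, two lower spine bones and a left glute cloth helper, while a large root box acts as a floor with its top near ground level.
# Leg, foot and pelvis helper bodies list no shapes, and there are no constraints.
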PhysicsAsset /BRCosmetics/Characters/Player/Male/Medium/Bodies/M_MED_RampCanal_Bar/Meshes/M_MED_RampCanal_Bar_ClothCollision
bodies {
 calf_l []
 calf_r []
 foot_l []
 foot_r []
 glute_l_cloth_helper [{shapeType=Capsule, length=2, radius=4.5, shapeName=deform_glute_l_capsule, center=<4.7 -2 1.25>, rotation=<rot 15 30 -95>}]
 pelvis [{shapeType=Capsule, length=9, radius=11, center=<-5 1.5 0>, rotation=<rot 0 -0.0006 0>}]
 pelvis_cloth_helper []
 root [{shapeType=Box, extentX=331.4972, extentY=332.5133, extentZ=73.6069, shapeName=root_box, center=<0 0 -36.7803>, rotation=<rot -0.0004 -90 0>}]
 spine_01 [{shapeType=Capsule, length=8.5, radius=9.5, center=<0.3095 2.5 0>}]
 spine_02 [{shapeType=Capsule, length=8.5, radius=8.5, center=<0.3095 2.2 0>}]
 thigh_l []
 thigh_r []
 thigh_twist_01_l []
}
constraints []
